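# Control Rig for the M_LRG_BakerStep large male body cosmetic, correcting shoulder, shoulder armor and shoulderpad bones as the arms move.
# Spherical pose readers on upper arms and clavicles weight additive bone offsets, mostly smoothed by AlphaInterp.
# SkyDive, Parachute, Pistol and Rifle variables weight extra adjustments, while Emote enables shoulder parent constraints.
# Separate deform_shoulderpad chains, apparently unwired, use distance remaps, pose readers and filtered constraints.
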ControlRigBlueprint /BRCosmetics/Characters/Player/Male/Large/Bodies/M_LRG_BakerStep/Meshes/M_LRG_BakerStep_CtrlRig
graphs [
 ContainedGraph {nodes [
   n1 = RigVMFunctionEntryNode()
   RigUnit_GetTransform = &RigUnit_GetTransform(Space=LocalSpace, bInitial=false, Item=@n1.Joint)
   n2 = &RigVMFunction_MathQuaternionToEuler(RotationOrder=ZYX, Value=@RigUnit_GetTransform.Transform.Rotation)
   > RigUnit_SetTransform = &RigUnit_SetTransform(
    Space=LocalSpace
    Value=(Rotation=(X=0.0132,Y=0.4099,Z=-0.0222,W=0.9118),Translation=(X=23.6045,Y=-6.37,Z=140.1115),Scale3D=(X=1,Y=1,Z=1))
    bInitial=false
    Item=@n1.Joint
    Weight=@n1.Weight
    Value.Translation.Z=&RigVMFunction_MathFloatAdd(B=@n1.Translate Z, A=@RigUnit_GetTransform.Transform.Translation.Z) {template="Add::Execute(in A,in B,out Result)"}.Result
    Value.Scale3D.Z=&RigVMFunction_MathFloatAdd(B=@n1.Scale Z, A=@RigUnit_GetTransform.Transform.Scale3D.Z) {template="Add::Execute(in A,in B,out Result)"}.Result
    Value.Translation.Y=&RigVMFunction_MathFloatAdd(B=@n1.Translate Y, A=@RigUnit_GetTransform.Transform.Translation.Y) {template="Add::Execute(in A,in B,out Result)"}.Result
    Value.Scale3D.Y=&RigVMFunction_MathFloatAdd(B=@n1.Scale Y, A=@RigUnit_GetTransform.Transform.Scale3D.Y) {template="Add::Execute(in A,in B,out Result)"}.Result
    Value.Translation.X=&RigVMFunction_MathFloatAdd(B=@n1.Translate X, A=@RigUnit_GetTransform.Transform.Translation.X) {template="Add::Execute(in A,in B,out Result)"}.Result
    Value.Scale3D.X=&RigVMFunction_MathFloatAdd(B=@n1.Scale X, A=@RigUnit_GetTransform.Transform.Scale3D.X) {template="Add::Execute(in A,in B,out Result)"}.Result
    Value.Rotation=&RigVMFunction_MathQuaternionFromEuler(
     Euler=(X=0,Y=0,Z=0)
     Euler.X=&RigVMFunction_MathFloatAdd(B=@n1.Rotate X, A=@n2.Result.X) {template="Add::Execute(in A,in B,out Result)"}.Result
     Euler.Y=&RigVMFunction_MathFloatAdd(B=@n1.Rotate Y, A=@n2.Result.Y) {template="Add::Execute(in A,in B,out Result)"}.Result
     Euler.Z=&RigVMFunction_MathFloatAdd(B=@n1.Rotate Z, A=@n2.Result.Z) {template="Add::Execute(in A,in B,out Result)"}.Result
    ).Result
   )
   > RigVMFunctionReturnNode()
  ]}
 ContainedGraph {nodes [
   n1 = RigVMFunctionEntryNode()
   SequenceExecution_3_1 = &RigVMFunction_Sequence(ExecuteContext=@n1.ExecuteContext)
   SequenceExecution_3_1_1 = &RigVMFunction_Sequence(ExecuteContext=@SequenceExecution_3_1.B)
   SequenceExecution_3_1_2 = &RigVMFunction_Sequence(ExecuteContext=@SequenceExecution_3_1_1.B)
   RigVMFunctionReturnNode(A=@SequenceExecution_3_1.A, B=@SequenceExecution_3_1_1.A, C=@SequenceExecution_3_1_2.A, D=@SequenceExecution_3_1_2.B)
  ]}
 New Function_ContainedGraph {nodes [
   RigVMFunctionEntryNode()
   > RigVMFunction_Sequence_3 = RigVMAggregateNode()
   n1 = RigVMVariableNode() {variable=SkyDive}
   AlphaInterp = &RigVMFunction_AlphaInterp(InterpSpeedDecreasing=8, InterpSpeedIncreasing=2, Scale=1, bClampResult=true, Value=@n1.Value)
   &RigUnit_ModifyTransforms(
    ItemToModify=((Item=(Type=Bone,Name="dyn_shoulder_MSTR"),Transform=(Rotation=(X=0,Y=0,Z=0,W=1),Translation=(X=0,Y=-3,Z=-3),Scale3D=(X=0.9,Y=0.9,Z=0.9))))
    Weight=@AlphaInterp.Result
    ItemToModify.0.Transform.Rotation=&RigVMFunction_MathQuaternionFromEuler(Euler=(X=0,Y=40,Z=0), RotationOrder=XYZ).Result
    ExecutePin=@RigVMFunction_Sequence_3.A
   )
   n2 = &RigVMFunction_AlphaInterp(InterpSpeedDecreasing=8, InterpSpeedIncreasing=2, Scale=1, bClampResult=true, Value=@n1.Value)
   > &RigUnit_ModifyTransforms(
    ItemToModify=((Item=(Type=Bone,Name="dyn_shoulder_r_MSTR"),Transform=(Rotation=(X=0,Y=0,Z=0,W=1),Translation=(X=0,Y=-3,Z=3),Scale3D=(X=0.9,Y=0.9,Z=0.9))))
    Weight=@n2.Result
    ItemToModify.0.Transform.Rotation=&RigVMFunction_MathQuaternionFromEuler(Euler=(X=0,Y=-40,Z=0), RotationOrder=XYZ).Result
   )
   n3 = &RigVMFunction_AlphaInterp(InterpSpeedDecreasing=10, InterpSpeedIncreasing=10, Scale=1, bClampResult=true, Value=RigVMVariableNode() {variable=Parachute}.Value)
   > &RigUnit_ModifyTransforms(
    ItemToModify=((Item=(Type=Bone,Name="dyn_shoulder_MSTR"),Transform=(Rotation=(X=0,Y=0,Z=0,W=1),Translation=(X=0,Y=-2,Z=6),Scale3D=(X=1,Y=1,Z=1))))
    Weight=@n3.Result
    ItemToModify.0.Transform.Rotation=&RigVMFunction_MathQuaternionFromEuler(Euler=(X=10,Y=-50,Z=-20), RotationOrder=XYZ).Result
   )
   n4 = &RigVMFunction_AlphaInterp(InterpSpeedDecreasing=10, InterpSpeedIncreasing=10, Scale=1, bClampResult=true, Value=RigVMVariableNode() {variable=Parachute}.Value)
   > &RigUnit_ModifyTransforms(
    ItemToModify=((Item=(Type=Bone,Name="dyn_shoulder_r_MSTR"),Transform=(Rotation=(X=0,Y=0,Z=0,W=1),Translation=(X=0,Y=-4,Z=-6),Scale3D=(X=1,Y=1,Z=1))))
    Weight=@n4.Result
    ItemToModify.0.Transform.Rotation=&RigVMFunction_MathQuaternionFromEuler(Euler=(X=-10,Y=40,Z=-20), RotationOrder=XYZ).Result
   )
   n5 = &RigVMFunction_Sequence(ExecuteContext=@RigVMFunction_Sequence_3.B)
   n6 = &RigUnit_SphericalPoseReader(
    ActiveRegionScaleFactors=(PositiveWidth=1.3,NegativeWidth=0.6,PositiveHeight=1,NegativeHeight=1.4)
    ActiveRegionSize=0.15
    Debug=(bDrawDebug=false,bDraw2D=false,bDrawLocalAxes=true,DebugScale=25,DebugSegments=20,DebugThickness=0.25)
    DriverAxis=(X=1,Y=0,Z=0)
    DriverItem=(Type=Bone,Name="upperarm_l")
    FalloffRegionScaleFactors=(PositiveWidth=0.6,NegativeWidth=0.6,PositiveHeight=1,NegativeHeight=1)
    FalloffSize=0.45
    OptionalParentItem=(Type=Bone,Name="spine_04")
    RotationOffset=(X=40,Y=25,Z=-40)
    ExecutePin=@n5.A
   )
   > &RigUnit_ModifyTransforms(
    ItemToModify=((Item=(Type=Bone,Name="dyn_shoulder_01"),Transform=(Rotation=(X=0,Y=0,Z=0,W=1),Translation=(X=0,Y=-3,Z=-2),Scale3D=(X=1,Y=1,Z=1))),(Item=(Type=Bone,Name="None"),Transform=(Rotation=(X=0,Y=0,Z=0,W=1),Translation=(X=0,Y=0,Z=0),Scale3D=(X=1,Y=1,Z=1))))
    ItemToModify.0.Transform.Rotation=&RigVMFunction_MathQuaternionFromEuler(Euler=(X=10,Y=30,Z=-10), RotationOrder=XYZ).Result
    Weight=@n6.OutputParam
   )
   > n7 = &RigUnit_SphericalPoseReader(
    ActiveRegionScaleFactors=(PositiveWidth=0.5,NegativeWidth=1,PositiveHeight=1,NegativeHeight=2)
    ActiveRegionSize=0.08
    Debug=(bDrawDebug=false,bDraw2D=false,bDrawLocalAxes=true,DebugScale=25,DebugSegments=20,DebugThickness=0.25)
    DriverAxis=(X=1,Y=0,Z=0)
    DriverItem=(Type=Bone,Name="upperarm_l")
    FalloffRegionScaleFactors=(PositiveWidth=0.5,NegativeWidth=0.2,PositiveHeight=0.2,NegativeHeight=1)
    FalloffSize=0.45
    OptionalParentItem=(Type=Bone,Name="spine_04")
    RotationOffset=(X=80,Y=30,Z=-25)
   )
   > &RigUnit_ModifyTransforms(
    ItemToModify=((Item=(Type=Bone,Name="dyn_shoulder_01"),Transform=(Rotation=(X=0,Y=0,Z=0,W=1),Translation=(X=0,Y=0,Z=0),Scale3D=(X=1,Y=1,Z=1))))
    ItemToModify.0.Transform.Rotation=&RigVMFunction_MathQuaternionFromEuler(Euler=(X=0,Y=0,Z=-5)).Result
    Weight=@n7.OutputParam
   )
   n8 = &RigUnit_SphericalPoseReader(
    ActiveRegionScaleFactors=(PositiveWidth=0.6,NegativeWidth=0.6,PositiveHeight=1,NegativeHeight=1)
    ActiveRegionSize=0.15
    Debug=(bDrawDebug=false,bDraw2D=false,bDrawLocalAxes=false,DebugScale=25,DebugSegments=20,DebugThickness=0.25)
    DriverAxis=(X=-1,Y=0,Z=0)
    DriverItem=(Type=Bone,Name="upperarm_r")
    FalloffRegionScaleFactors=(PositiveWidth=0.6,NegativeWidth=0.6,PositiveHeight=2,NegativeHeight=1)
    FalloffSize=0.45
    FlipHeightScaling=false
    FlipWidthScaling=true
    OptionalParentItem=(Type=Bone,Name="spine_04")
    RotationOffset=(X=-100,Y=25,Z=40)
    ExecutePin=@n5.B
   )
   > &RigUnit_ModifyTransforms(
    ItemToModify=((Item=(Type=Bone,Name="dyn_shoulder_r_01"),Transform=(Rotation=(X=0,Y=0,Z=0,W=1),Translation=(X=0,Y=0,Z=0),Scale3D=(X=1,Y=1,Z=1))),(Item=(Type=Bone,Name="None"),Transform=(Rotation=(X=0,Y=0,Z=0,W=1),Translation=(X=0,Y=0,Z=0),Scale3D=(X=1,Y=1,Z=1))))
    ItemToModify.0.Transform.Rotation=&RigVMFunction_MathQuaternionFromEuler(Euler=(X=0,Y=-15,Z=10), RotationOrder=XYZ).Result
    Weight=@n8.OutputParam
   )
   > n9 = &RigUnit_SphericalPoseReader(
    ActiveRegionScaleFactors=(PositiveWidth=1,NegativeWidth=1,PositiveHeight=2,NegativeHeight=1)
    ActiveRegionSize=0.05
    Debug=(bDrawDebug=false,bDraw2D=false,bDrawLocalAxes=true,DebugScale=25,DebugSegments=20,DebugThickness=0.25)
    DriverAxis=(X=-1,Y=0,Z=0)
    DriverItem=(Type=Bone,Name="upperarm_r")
    FalloffRegionScaleFactors=(PositiveWidth=0.5,NegativeWidth=0.5,PositiveHeight=1,NegativeHeight=1)
    FalloffSize=0.3
    FlipWidthScaling=true
    OptionalParentItem=(Type=Bone,Name="spine_04")
    RotationOffset=(X=-50,Y=35,Z=0)
   )
   n10 = &RigVMFunction_AlphaInterp(
    InterpSpeedDecreasing=10
    InterpSpeedIncreasing=10
    bClampResult=False
    bInterpResult=False
    Value=@n9.OutputParam
    Scale=RigVMVariableNode() {variable=Pistol}.Value
   )
   > &RigUnit_ModifyTransforms(
    ItemToModify=((Item=(Type=Bone,Name="dyn_shoulder_r_01"),Transform=(Rotation=(X=0,Y=0,Z=0,W=1),Translation=(X=-1,Y=2,Z=-3),Scale3D=(X=1,Y=1,Z=1))),(Item=(Type=Bone,Name="None"),Transform=(Rotation=(X=0,Y=0,Z=0,W=1),Translation=(X=0,Y=0,Z=0),Scale3D=(X=1,Y=1,Z=1))))
    Weight=@n10.Result
    ItemToModify.0.Transform.Rotation=&RigVMFunction_MathQuaternionFromEuler(Euler=(X=10,Y=20,Z=20), RotationOrder=XYZ).Result
   )
   > n11 = &RigUnit_SphericalPoseReader(
    ActiveRegionScaleFactors=(PositiveWidth=0.3,NegativeWidth=0.2,PositiveHeight=0.2,NegativeHeight=0.25)
    ActiveRegionSize=0.5
    Debug=(bDrawDebug=false,DebugScale=25,DebugSegments=20,DebugThickness=0.25)
    DriverAxis=(X=-1,Y=0,Z=0)
    DriverItem=(Type=Bone,Name="upperarm_r")
    FalloffRegionScaleFactors=(PositiveWidth=0.4,NegativeWidth=0.1,PositiveHeight=0.1,NegativeHeight=0.3)
    FalloffSize=0.3
    OptionalParentItem=(Type=Bone,Name="spine_05")
    RotationOffset=(X=-10,Y=50,Z=0)
   )
   n12 = &RigVMFunction_AlphaInterp(
    InterpSpeedDecreasing=12
    InterpSpeedIncreasing=15
    bClampResult=false
    Value=@n11.OutputParam
    Scale=RigVMVariableNode() {variable=Rifle}.Value
   )
   > &RigUnit_ModifyTransforms(
    ItemToModify=((Item=(Type=Bone,Name="dyn_shoulder_r_01"),Transform=(Rotation=(X=0,Y=0,Z=0,W=1),Translation=(X=-1,Y=0,Z=0),Scale3D=(X=1,Y=1,Z=1))),(Item=(Type=Bone,Name="None"),Transform=(Rotation=(X=0,Y=0,Z=0,W=1),Translation=(X=0,Y=0,Z=0),Scale3D=(X=1,Y=1,Z=1))))
    Weight=@n12.Result
    ItemToModify.0.Transform.Rotation=&RigVMFunction_MathQuaternionFromEuler(Euler=(X=0,Y=0,Z=0), RotationOrder=XYZ).Result
   )
   > n13 = &RigUnit_SphericalPoseReader(
    ActiveRegionScaleFactors=(PositiveWidth=0.5,NegativeWidth=1,PositiveHeight=1.8,NegativeHeight=2)
    ActiveRegionSize=0.08
    Debug=(bDrawDebug=false,bDraw2D=false,bDrawLocalAxes=false,DebugScale=25,DebugSegments=20,DebugThickness=0.25)
    DriverAxis=(X=-1,Y=0,Z=0)
    DriverItem=(Type=Bone,Name="upperarm_r")
    FalloffRegionScaleFactors=(PositiveWidth=0.5,NegativeWidth=1,PositiveHeight=2,NegativeHeight=1)
    FalloffSize=0.45
    FlipHeightScaling=false
    FlipWidthScaling=true
    OptionalParentItem=(Type=Bone,Name="spine_04")
    RotationOffset=(X=-80,Y=40,Z=25)
   )
   > &RigUnit_ModifyTransforms(
    ItemToModify=((Item=(Type=Bone,Name="dyn_shoulder_r_01"),Transform=(Rotation=(X=0,Y=0,Z=0,W=1),Translation=(X=1,Y=0,Z=0),Scale3D=(X=1,Y=1,Z=1))),(Item=(Type=Bone,Name="None"),Transform=(Rotation=(X=0,Y=0,Z=0,W=1),Translation=(X=0,Y=0,Z=0),Scale3D=(X=1,Y=1,Z=1))))
    ItemToModify.0.Transform.Rotation=&RigVMFunction_MathQuaternionFromEuler(Euler=(X=0,Y=25,Z=0), RotationOrder=XYZ).Result
    Weight=@n13.OutputParam
   )
   RigVMFunction_ControlFlowBranch = &RigVMFunction_ControlFlowBranch(ExecuteContext=@RigVMFunction_Sequence_3.C, Condition=RigVMVariableNode() {variable=Emote}.Value)
   n14 = &RigVMFunction_AlphaInterp(InterpSpeedDecreasing=7, InterpSpeedIncreasing=7, Scale=1, bClampResult=False, Value=RigVMVariableNode() {variable=EmoteAlpha}.Value)
   &RigUnit_ParentConstraint(
    AdvancedSettings=(RotationOrderForFilter=XZY)
    Child=(Type=Bone,Name="dyn_shoulder_r_01")
    Parents=((Item=(Type=Bone,Name="upperarm_twist_01_r"),Weight=0.5),(Item=(Type=Bone,Name="clavicle_r"),Weight=0.5))
    Weight=@n14.Result
    ExecutePin=@RigVMFunction_ControlFlowBranch.True
   )
   > &RigUnit_ParentConstraint(
    AdvancedSettings=(RotationOrderForFilter=XZY)
    Child=(Type=Bone,Name="dyn_shoulder_01")
    Parents=((Item=(Type=Bone,Name="upperarm_l"),Weight=0.5),(Item=(Type=Bone,Name="clavicle_l"),Weight=0.5))
    Weight=@n14.Result
   )
   > SequenceExecution_3_1_1 = RigVMAggregateNode()
   n15 = &RigUnit_SphericalPoseReader(
    ActiveRegionScaleFactors=(PositiveWidth=0.5,NegativeWidth=0.5,PositiveHeight=1,NegativeHeight=1)
    ActiveRegionSize=0.15
    Debug=(bDrawDebug=true,bDraw2D=false,bDrawLocalAxes=false,DebugScale=25,DebugSegments=20,DebugThickness=0.25)
    DriverAxis=(X=1,Y=0,Z=0)
    DriverItem=(Type=Bone,Name="upperarm_l")
    FalloffRegionScaleFactors=(PositiveWidth=1,NegativeWidth=1,PositiveHeight=1,NegativeHeight=1)
    FalloffSize=0.4
    FlipHeightScaling=false
    FlipWidthScaling=true
    OptionalParentItem=(Type=Bone)
    RotationOffset=(X=-15,Y=-40,Z=10)
    ExecutePin=@SequenceExecution_3_1_1.A
   )
   n16 = &RigVMFunction_AlphaInterp(InterpSpeedDecreasing=100, InterpSpeedIncreasing=100, Scale=1, bClampResult=False, Value=@n15.OutputParam)
   > &RigUnit_ModifyTransforms(
    ItemToModify=((Item=(Type=Bone,Name="dyn_shoulder_01"),Transform=(Rotation=(X=0,Y=0,Z=0,W=1),Translation=(X=3,Y=-4,Z=0),Scale3D=(X=1,Y=1,Z=1))))
    Weight=@n16.Result
    ItemToModify.0.Transform.Rotation=&RigVMFunction_MathQuaternionFromEuler(Euler=(X=0,Y=0,Z=-15), RotationOrder=ZXY).Result
   )
   > n17 = &RigUnit_SphericalPoseReader(
    ActiveRegionScaleFactors=(PositiveWidth=1,NegativeWidth=1,PositiveHeight=0.5,NegativeHeight=0.5)
    ActiveRegionSize=0.15
    Debug=(bDrawDebug=false,bDraw2D=false,bDrawLocalAxes=false,DebugScale=25,DebugSegments=20,DebugThickness=0.25)
    DriverAxis=(X=-1,Y=0,Z=0)
    DriverItem=(Type=Bone,Name="upperarm_r")
    FalloffRegionScaleFactors=(PositiveWidth=1,NegativeWidth=1,PositiveHeight=1,NegativeHeight=1)
    FalloffSize=0.4
    FlipHeightScaling=false
    FlipWidthScaling=true
    OptionalParentItem=(Type=Bone)
    RotationOffset=(X=-15,Y=140,Z=-10)
   )
   n18 = &RigVMFunction_AlphaInterp(InterpSpeedDecreasing=100, InterpSpeedIncreasing=100, Scale=1, bClampResult=False, Value=@n17.OutputParam)
   > &RigUnit_ModifyTransforms(
    ItemToModify=((Item=(Type=Bone,Name="dyn_shoulder_r_01"),Transform=(Rotation=(X=0,Y=0,Z=0,W=1),Translation=(X=3,Y=-4,Z=0),Scale3D=(X=1,Y=1,Z=1))))
    Weight=@n18.Result
    ItemToModify.0.Transform.Rotation=&RigVMFunction_MathQuaternionFromEuler(Euler=(X=0,Y=0,Z=0), RotationOrder=ZXY).Result
   )
   n19 = &RigUnit_SphericalPoseReader(
    ActiveRegionScaleFactors=(PositiveWidth=1,NegativeWidth=1,PositiveHeight=1,NegativeHeight=1)
    ActiveRegionSize=0.1
    Debug=(bDrawDebug=false,bDraw2D=false,bDrawLocalAxes=false,DebugScale=25,DebugSegments=20,DebugThickness=0.25)
    DriverAxis=(X=1,Y=0,Z=0)
    DriverItem=(Type=Bone,Name="upperarm_l")
    FalloffRegionScaleFactors=(PositiveWidth=1,NegativeWidth=1,PositiveHeight=1,NegativeHeight=1)
    FalloffSize=0.65
    FlipWidthScaling=true
    OptionalParentItem=(Type=Bone)
    RotationOffset=(X=0,Y=220,Z=0)
    ExecutePin=@SequenceExecution_3_1_1.B
   )
   n20 = &RigVMFunction_AlphaInterp(InterpSpeedDecreasing=100, InterpSpeedIncreasing=100, Scale=1, bClampResult=False, Value=@n19.OutputParam)
   > &RigUnit_ModifyTransforms(
    ItemToModify=((Item=(Type=Bone,Name="dyn_shoulder_armor_l"),Transform=(Rotation=(X=0,Y=0,Z=0,W=1),Translation=(X=0,Y=0,Z=0),Scale3D=(X=1,Y=1,Z=1))))
    Weight=@n20.Result
    ItemToModify.0.Transform.Rotation=&RigVMFunction_MathQuaternionFromEuler(Euler=(X=25,Y=-5,Z=0), RotationOrder=ZXY).Result
   )
   > n21 = &RigUnit_SphericalPoseReader(
    ActiveRegionScaleFactors=(PositiveWidth=1,NegativeWidth=0.35,PositiveHeight=0.25,NegativeHeight=0.45)
    ActiveRegionSize=0.2
    Debug=(bDrawDebug=false,bDraw2D=false,bDrawLocalAxes=false,DebugScale=25,DebugSegments=20,DebugThickness=0.25)
    DriverAxis=(X=-1,Y=0,Z=0)
    DriverItem=(Type=Bone,Name="clavicle_r")
    FalloffRegionScaleFactors=(PositiveWidth=1,NegativeWidth=1,PositiveHeight=1,NegativeHeight=1)
    FalloffSize=0.25
    FlipWidthScaling=true
    OptionalParentItem=(Type=Bone)
    RotationOffset=(X=-120,Y=0,Z=-80)
   )
   n22 = &RigVMFunction_AlphaInterp(InterpSpeedDecreasing=100, InterpSpeedIncreasing=100, Scale=1, bClampResult=False, Value=@n21.OutputParam)
   > &RigUnit_ModifyTransforms(
    ItemToModify=((Item=(Type=Bone,Name="dyn_shoulder_armor_r"),Transform=(Rotation=(X=0,Y=0,Z=0,W=1),Translation=(X=-2,Y=0,Z=1),Scale3D=(X=1,Y=1,Z=1))))
    Weight=@n22.Result
    ItemToModify.0.Transform.Rotation=&RigVMFunction_MathQuaternionFromEuler(Euler=(X=40,Y=15,Z=20), RotationOrder=ZXY).Result
   )
   n23 = &RigUnit_SphericalPoseReader(
    ActiveRegionScaleFactors=(PositiveWidth=1,NegativeWidth=0.5,PositiveHeight=0.5,NegativeHeight=0.5)
    ActiveRegionSize=0.15
    Debug=(bDrawDebug=false,bDraw2D=false,bDrawLocalAxes=false,DebugScale=25,DebugSegments=20,DebugThickness=0.25)
    DriverAxis=(X=1,Y=0,Z=0)
    DriverItem=(Type=Bone,Name="upperarm_l")
    FalloffRegionScaleFactors=(PositiveWidth=1,NegativeWidth=1,PositiveHeight=1,NegativeHeight=1)
    FalloffSize=0.65
    FlipHeightScaling=false
    FlipWidthScaling=true
    OptionalParentItem=(Type=Bone)
    RotationOffset=(X=-30,Y=30,Z=-90)
    ExecutePin=@SequenceExecution_3_1_1.C
   )
   n24 = &RigVMFunction_AlphaInterp(InterpSpeedDecreasing=100, InterpSpeedIncreasing=1000, Scale=1, bClampResult=False, Value=@n23.OutputParam)
   > &RigUnit_ModifyTransforms(
    ItemToModify=((Item=(Type=Bone,Name="dyn_shoulder_armor_l"),Transform=(Rotation=(X=0,Y=0,Z=0,W=1),Translation=(X=2,Y=3,Z=3),Scale3D=(X=1,Y=1,Z=1))))
    Weight=@n24.Result
    ItemToModify.0.Transform.Rotation=&RigVMFunction_MathQuaternionFromEuler(Euler=(X=-60,Y=20,Z=10), RotationOrder=ZXY).Result
   )
   > n25 = &RigUnit_SphericalPoseReader(
    ActiveRegionScaleFactors=(PositiveWidth=0.5,NegativeWidth=1,PositiveHeight=0.5,NegativeHeight=0.5)
    ActiveRegionSize=0.15
    Debug=(bDrawDebug=false,bDraw2D=false,bDrawLocalAxes=false,DebugScale=25,DebugSegments=20,DebugThickness=0.25)
    DriverAxis=(X=-1,Y=0,Z=0)
    DriverItem=(Type=Bone,Name="upperarm_r")
    FalloffRegionScaleFactors=(PositiveWidth=1,NegativeWidth=1,PositiveHeight=1,NegativeHeight=1)
    FalloffSize=0.65
    FlipHeightScaling=false
    FlipWidthScaling=true
    OptionalParentItem=(Type=Bone)
    RotationOffset=(X=-35,Y=150,Z=90)
   )
   n26 = &RigVMFunction_AlphaInterp(InterpSpeedDecreasing=100, InterpSpeedIncreasing=1000, Scale=1, bClampResult=False, Value=@n25.OutputParam)
   > &RigUnit_ModifyTransforms(
    ItemToModify=((Item=(Type=Bone,Name="dyn_shoulder_armor_r"),Transform=(Rotation=(X=0,Y=0,Z=0,W=1),Translation=(X=2,Y=0,Z=0),Scale3D=(X=1,Y=1,Z=1))))
    Weight=@n26.Result
    ItemToModify.0.Transform.Rotation=&RigVMFunction_MathQuaternionFromEuler(Euler=(X=10,Y=-20,Z=10), RotationOrder=ZXY).Result
   )
   n27 = &RigUnit_SphericalPoseReader(
    ActiveRegionScaleFactors=(PositiveWidth=0.35,NegativeWidth=1,PositiveHeight=0.25,NegativeHeight=0.45)
    ActiveRegionSize=0.2
    Debug=(bDrawDebug=false,bDraw2D=false,bDrawLocalAxes=false,DebugScale=25,DebugSegments=20,DebugThickness=0.25)
    DriverAxis=(X=1,Y=0,Z=0)
    DriverItem=(Type=Bone,Name="clavicle_l")
    FalloffRegionScaleFactors=(PositiveWidth=1,NegativeWidth=1,PositiveHeight=1,NegativeHeight=1)
    FalloffSize=0.25
    FlipWidthScaling=true
    OptionalParentItem=(Type=Bone)
    RotationOffset=(X=60,Y=0,Z=-100)
    ExecutePin=@SequenceExecution_3_1_1.D
   )
   n28 = &RigVMFunction_AlphaInterp(InterpSpeedDecreasing=100, InterpSpeedIncreasing=100, Scale=1, bClampResult=False, Value=@n27.OutputParam)
   > &RigUnit_ModifyTransforms(
    ItemToModify=((Item=(Type=Bone,Name="dyn_shoulder_armor_l"),Transform=(Rotation=(X=0,Y=0,Z=0,W=1),Translation=(X=-2,Y=0,Z=-1),Scale3D=(X=1,Y=1,Z=1))))
    Weight=@n28.Result
    ItemToModify.0.Transform.Rotation=&RigVMFunction_MathQuaternionFromEuler(Euler=(X=40,Y=-30,Z=-20), RotationOrder=ZXY).Result
   )
   > n29 = &RigUnit_SphericalPoseReader(
    ActiveRegionScaleFactors=(PositiveWidth=1,NegativeWidth=1,PositiveHeight=1,NegativeHeight=1)
    ActiveRegionSize=0.1
    Debug=(bDrawDebug=false,bDraw2D=false,bDrawLocalAxes=false,DebugScale=25,DebugSegments=20,DebugThickness=0.25)
    DriverAxis=(X=-1,Y=0,Z=0)
    DriverItem=(Type=Bone,Name="upperarm_r")
    FalloffRegionScaleFactors=(PositiveWidth=1,NegativeWidth=1,PositiveHeight=1,NegativeHeight=1)
    FalloffSize=0.65
    FlipWidthScaling=true
    OptionalParentItem=(Type=Bone)
    RotationOffset=(X=180,Y=220,Z=0)
   )
   n30 = &RigVMFunction_AlphaInterp(InterpSpeedDecreasing=100, InterpSpeedIncreasing=100, Scale=1, bClampResult=False, Value=@n29.OutputParam)
   > &RigUnit_ModifyTransforms(
    ItemToModify=((Item=(Type=Bone,Name="dyn_shoulder_armor_r"),Transform=(Rotation=(X=0,Y=0,Z=0,W=1),Translation=(X=0,Y=0,Z=1),Scale3D=(X=1,Y=1,Z=1))))
    Weight=@n30.Result
    ItemToModify.0.Transform.Rotation=&RigVMFunction_MathQuaternionFromEuler(Euler=(X=-25,Y=0,Z=5), RotationOrder=ZXY).Result
   )
   RigVMFunctionReturnNode(ExecuteContext=@RigVMFunction_Sequence_3.D)
   &RigUnit_ModifyTransforms(
    ExecutePin=()
    ItemToModify=((Item=(Type=Bone,Name="clavicle_l"),Transform=(Rotation=(X=0,Y=0,Z=0,W=1),Translation=(X=0,Y=0,Z=0),Scale3D=(X=1,Y=1,Z=1))))
    Weight=1
    WeightMinimum=0
   )
   > &RigUnit_ModifyTransforms(
    ItemToModify=((Item=(Type=Bone,Name="upperarm_l"),Transform=(Rotation=(X=0,Y=0,Z=0,W=1),Translation=(X=0,Y=0,Z=0),Scale3D=(X=1,Y=1,Z=1))))
    Weight=1
    WeightMinimum=0
   )
   > &RigUnit_ModifyTransforms(
    ItemToModify=((Item=(Type=Bone,Name="lowerarm_l"),Transform=(Rotation=(X=0,Y=0,Z=0,W=1),Translation=(X=0,Y=0,Z=0),Scale3D=(X=1,Y=1,Z=1))))
    Weight=1
    WeightMinimum=0
   )
   RigVMFunction_MathVectorDistance = &RigVMFunction_MathVectorDistance(
    A=&RigUnit_GetTransform(Item=(Type=Bone,Name="deform_shoulderpad_target_up_l"), bInitial=False).Transform.Translation
    B=&RigUnit_GetTransform(Item=(Type=Bone,Name="head"), bInitial=False).Transform.Translation
   )
   n31 = &RigVMFunction_AlphaInterp(
    InterpSpeedDecreasing=30
    InterpSpeedIncreasing=30
    Scale=1
    bClampResult=true
    Value=&RigVMFunction_MathDoubleRemap(SourceMaximum=8, SourceMinimum=25, TargetMaximum=1, TargetMinimum=0, Value=@RigVMFunction_MathVectorDistance.Result).Result
   )
   > &RigUnit_ModifyTransforms(
    ItemToModify=((Item=(Type=Bone,Name="deform_shoulderpad_l"),Transform=(Rotation=(X=0,Y=0.2334,Z=0,W=0.9724),Translation=(X=4,Y=0,Z=0),Scale3D=(X=0.9,Y=1,Z=0.9))))
    WeightMinimum=0
    Weight=@n31.Result
   )
   n32 = &RigUnit_GetRelativeTransformForItem(
    Child=(Type=Bone,Name="deform_shoulderpad_target_side_l")
    Parent=(Type=Bone,Name="upperarm_twist_02_l")
    bChildInitial=false
    bParentInitial=false
   )
   n33 = &RigVMFunction_AlphaInterp(
    InterpSpeedDecreasing=30
    InterpSpeedIncreasing=30
    Scale=1
    bClampResult=true
    Value=&RigVMFunction_MathDoubleRemap(SourceMaximum=-12, SourceMinimum=9, TargetMaximum=1, TargetMinimum=0, Value=@n32.RelativeTransform.Translation.Z).Result
   )
   > &RigUnit_ModifyTransforms(
    ItemToModify=((Item=(Type=Bone,Name="deform_shoulderpad_l"),Transform=(Rotation=(X=0,Y=-0.3827,Z=0,W=0.9239),Translation=(X=1,Y=0,Z=9),Scale3D=(X=0.9,Y=1,Z=0.9))))
    WeightMinimum=0
    Weight=@n33.Result
   )
   > Get Delta Ref_2 = RigVMFunctionReferenceNode(Item=(Type=Bone,Name="clavicle_l"))
   n34 = &RigVMFunction_AlphaInterp(
    InterpSpeedDecreasing=30
    InterpSpeedIncreasing=30
    Scale=1
    bClampResult=true
    Value=&RigVMFunction_MathDoubleRemap(SourceMaximum=30, SourceMinimum=0, TargetMaximum=1, TargetMinimum=0, Value=@"Get Delta Ref_2".Result.Rotation.Yaw).Result
   )
   > &RigUnit_ModifyTransforms(
    ItemToModify=((Item=(Type=Bone,Name="deform_shoulderpad_l_base_root"),Transform=(Rotation=(X=0,Y=0,Z=-0.1478,W=0.989),Translation=(X=0,Y=-1,Z=0),Scale3D=(X=1,Y=1,Z=1))))
    WeightMinimum=0
    Weight=@n34.Result
   )
   n35 = &RigVMFunction_AlphaInterp(
    InterpSpeedDecreasing=30
    InterpSpeedIncreasing=30
    Scale=1
    bClampResult=true
    Value=&RigVMFunction_MathDoubleRemap(SourceMaximum=-30, SourceMinimum=0, TargetMaximum=1, TargetMinimum=0, Value=@"Get Delta Ref_2".Result.Rotation.Yaw).Result
   )
   > &RigUnit_ModifyTransforms(
    ItemToModify=((Item=(Type=Bone,Name="deform_shoulderpad_l_base_root"),Transform=(Rotation=(X=0,Y=0,Z=0.2164,W=0.9763),Translation=(X=2,Y=0,Z=0),Scale3D=(X=1,Y=1,Z=1))))
    WeightMinimum=0
    Weight=@n35.Result
   )
   n36 = &RigVMFunction_AlphaInterp(
    InterpSpeedDecreasing=30
    InterpSpeedIncreasing=30
    Scale=1
    bClampResult=true
    Value=&RigVMFunction_MathDoubleRemap(SourceMaximum=30, SourceMinimum=0, TargetMaximum=1, TargetMinimum=0, Value=@"Get Delta Ref_2".Result.Rotation.Pitch).Result
   )
   > &RigUnit_ModifyTransforms(
    ItemToModify=((Item=(Type=Bone,Name="deform_shoulderpad_l"),Transform=(Rotation=(X=0,Y=0.0436,Z=0,W=0.999),Translation=(X=0,Y=0,Z=0),Scale3D=(X=0.95,Y=0.95,Z=0.95))))
    WeightMinimum=0
    Weight=@n36.Result
   )
   n37 = &RigUnit_GetRelativeTransformForItem(
    Child=(Type=Bone,Name="deform_shoulderpad_target_side_l")
    Parent=(Type=Bone,Name="lowerarm_twist_02_l")
    bChildInitial=False
    bParentInitial=False
   )
   n38 = &RigVMFunction_AlphaInterp(
    InterpSpeedDecreasing=30
    InterpSpeedIncreasing=30
    Scale=1
    bClampResult=true
    Value=&RigVMFunction_MathDoubleRemap(SourceMaximum=1, SourceMinimum=-10, TargetMaximum=1, TargetMinimum=0, Value=@n37.RelativeTransform.Translation.X).Result
   )
   > &RigUnit_ModifyTransforms(
    ItemToModify=((Item=(Type=Bone,Name="deform_shoulderpad_l"),Transform=(Rotation=(X=0,Y=0.0436,Z=0,W=0.999),Translation=(X=-2,Y=0,Z=2),Scale3D=(X=0.95,Y=1,Z=0.95))))
    WeightMinimum=0
    Weight=@n38.Result
   )
   n39 = &RigUnit_GetRelativeTransformForItem(Child=(Type=Bone,Name="lowerarm_l"), Parent=(Type=Bone,Name="clavicle_l"), bChildInitial=False, bParentInitial=false)
   n40 = &RigVMFunction_AlphaInterp(
    InterpSpeedDecreasing=30
    InterpSpeedIncreasing=30
    Scale=1
    bClampResult=true
    Value=&RigVMFunction_MathDoubleRemap(SourceMaximum=25, SourceMinimum=0, TargetMaximum=1, TargetMinimum=0, Value=@n39.RelativeTransform.Translation.Y).Result
   )
   > &RigUnit_ModifyTransforms(
    ItemToModify=((Item=(Type=Bone,Name="deform_shoulderpad_l_base_root"),Transform=(Rotation=(X=0.2565,Y=-0.0534,Z=-0.0308,W=0.9646),Translation=(X=0,Y=1,Z=0),Scale3D=(X=1,Y=1,Z=1))))
    WeightMinimum=0
    Weight=@n40.Result
   )
   n41 = &RigVMFunction_AlphaInterp(
    InterpSpeedDecreasing=30
    InterpSpeedIncreasing=30
    Scale=1
    bClampResult=true
    Value=&RigVMFunction_MathDoubleRemap(SourceMaximum=-25, SourceMinimum=0, TargetMaximum=1, TargetMinimum=0, Value=@n39.RelativeTransform.Translation.Y).Result
   )
   > &RigUnit_ModifyTransforms(
    ItemToModify=((Item=(Type=Bone,Name="deform_shoulderpad_l_base_root"),Transform=(Rotation=(X=-0.2578,Y=-0.0226,Z=0.0842,W=0.9623),Translation=(X=0,Y=0,Z=0),Scale3D=(X=1,Y=1,Z=1))))
    WeightMinimum=0
    Weight=@n41.Result
   )
   > n42 = &RigUnit_SphericalPoseReader(
    ActiveRegionScaleFactors=(PositiveWidth=1,NegativeWidth=1,PositiveHeight=1,NegativeHeight=1)
    ActiveRegionSize=0.07
    Debug=(bDrawDebug=false,bDraw2D=False,bDrawLocalAxes=False,DebugScale=25,DebugSegments=20,DebugThickness=0.25)
    DriverAxis=(X=1,Y=0,Z=0)
    DriverItem=(Type=Bone,Name="upperarm_l")
    FalloffRegionScaleFactors=(PositiveWidth=1,NegativeWidth=1,PositiveHeight=1,NegativeHeight=1)
    FalloffSize=0.25
    FlipHeightScaling=False
    FlipWidthScaling=False
    OptionalParentItem=(Type=Bone,Name="clavicle_l")
    RotationOffset=(X=12,Y=12,Z=0)
   )
   > &RigUnit_ModifyTransforms(
    ItemToModify=((Item=(Type=Bone,Name="deform_shoulderpad_l"),Transform=(Rotation=(X=0.3409,Y=0.0473,Z=-0.1139,W=0.932),Translation=(X=-3.2797,Y=-2.8982,Z=0.0399),Scale3D=(X=1,Y=1,Z=1))))
    WeightMinimum=0
    Weight=@n42.OutputParam
   )
   > n43 = &RigUnit_SphericalPoseReader(
    ActiveRegionScaleFactors=(PositiveWidth=1,NegativeWidth=1,PositiveHeight=1,NegativeHeight=1)
    ActiveRegionSize=0.025
    Debug=(bDrawDebug=false,bDraw2D=False,bDrawLocalAxes=False,DebugScale=25,DebugSegments=20,DebugThickness=0.25)
    DriverAxis=(X=1,Y=0,Z=0)
    DriverItem=(Type=Bone,Name="lowerarm_l")
    FalloffRegionScaleFactors=(PositiveWidth=1,NegativeWidth=1,PositiveHeight=1,NegativeHeight=1)
    FalloffSize=0.2
    FlipHeightScaling=False
    FlipWidthScaling=False
    OptionalParentItem=(Type=Bone,Name="upperarm_l")
    RotationOffset=(X=0,Y=-95,Z=-5)
   )
   > n44 = &RigUnit_SphericalPoseReader(
    ActiveRegionScaleFactors=(PositiveWidth=1,NegativeWidth=1,PositiveHeight=1,NegativeHeight=1)
    ActiveRegionSize=0.15
    Debug=(bDrawDebug=false,bDraw2D=False,bDrawLocalAxes=False,DebugScale=25,DebugSegments=20,DebugThickness=0.25)
    DriverAxis=(X=1,Y=0,Z=0)
    DriverItem=(Type=Bone,Name="upperarm_l")
    FalloffRegionScaleFactors=(PositiveWidth=1,NegativeWidth=1,PositiveHeight=1,NegativeHeight=1)
    FalloffSize=0.3
    FlipHeightScaling=False
    FlipWidthScaling=False
    OptionalParentItem=(Type=Bone,Name="clavicle_l")
    RotationOffset=(X=0,Y=-35,Z=0)
   )
   n45 = &RigVMFunction_MathDoubleMul(B=1, A=&RigVMFunction_MathDoubleMul(B=@n43.OutputParam, A=@n44.OutputParam).Result)
   > &RigUnit_ParentConstraint(
    AdvancedSettings=(InterpolationType=Average,RotationOrderForFilter=XYZ)
    Child=(Type=Bone,Name="deform_shoulderpad_l_base_root")
    Filter=(TranslationFilter=(bX=false,bY=true,bZ=false),RotationFilter=(bX=false,bY=false,bZ=false),ScaleFilter=(bX=false,bY=false,bZ=false))
    Parents=((Item=(Type=Bone,Name="upperarm_l"),Weight=0.5),(Item=(Type=Bone,Name="lowerarm_l"),Weight=1))
    bMaintainOffset=true
    Weight=@n45.Result
   )
   > &RigUnit_ParentConstraint(
    AdvancedSettings=(InterpolationType=Average,RotationOrderForFilter=XYZ)
    Child=(Type=Bone,Name="deform_shoulderpad_l_base_root")
    Filter=(TranslationFilter=(bX=false,bY=false,bZ=false),RotationFilter=(bX=false,bY=true,bZ=true),ScaleFilter=(bX=false,bY=false,bZ=false))
    Parents=((Item=(Type=Bone,Name="clavicle_l"),Weight=0.8),(Item=(Type=Bone,Name="upperarm_l"),Weight=1))
    bMaintainOffset=false
    Weight=@n45.Result
   )
   > &RigUnit_ModifyTransforms(
    ItemToModify=((Item=(Type=Bone,Name="deform_shoulderpad_l_base_root"),Transform=(Rotation=(X=0.0014,Y=0.0261,Z=-0.0523,W=0.9983),Translation=(X=2,Y=0,Z=0),Scale3D=(X=1,Y=1,Z=1))))
    WeightMinimum=0
    Weight=@n45.Result
   )
   &RigUnit_ModifyTransforms(
    ExecutePin=()
    ItemToModify=((Item=(Type=Bone,Name="clavicle_r"),Transform=(Rotation=(X=0,Y=0,Z=0,W=1),Translation=(X=0,Y=0,Z=0),Scale3D=(X=1,Y=1,Z=1))))
    Weight=1
    WeightMinimum=0
   )
   > &RigUnit_ModifyTransforms(
    ItemToModify=((Item=(Type=Bone,Name="upperarm_r"),Transform=(Rotation=(X=0,Y=0,Z=0,W=1),Translation=(X=0,Y=0,Z=0),Scale3D=(X=1,Y=1,Z=1))))
    Weight=1
    WeightMinimum=0
   )
   > &RigUnit_ModifyTransforms(
    ItemToModify=((Item=(Type=Bone,Name="lowerarm_r"),Transform=(Rotation=(X=0,Y=0,Z=0,W=1),Translation=(X=0,Y=0,Z=0),Scale3D=(X=1,Y=1,Z=1))))
    Weight=1
    WeightMinimum=0
   )
   n46 = &RigUnit_GetTransform(Item=(Type=Bone,Name="head"), bInitial=False)
   n47 = &RigUnit_GetTransform(Item=(Type=Bone,Name="deform_shoulderpad_target_up_r"), bInitial=False)
   n48 = &RigVMFunction_AlphaInterp(
    InterpSpeedDecreasing=30
    InterpSpeedIncreasing=30
    Scale=1
    bClampResult=true
    Value=&RigVMFunction_MathDoubleRemap(
     SourceMaximum=8
     SourceMinimum=25
     TargetMaximum=1
     TargetMinimum=0
     Value=&RigVMFunction_MathVectorDistance(B=@n46.Transform.Translation, A=@n47.Transform.Translation).Result
    ).Result
   )
   > &RigUnit_ModifyTransforms(
    ItemToModify=((Item=(Type=Bone,Name="deform_shoulderpad_r"),Transform=(Rotation=(X=0,Y=0.2164,Z=0,W=0.9763),Translation=(X=-4,Y=0,Z=0),Scale3D=(X=0.9,Y=1,Z=0.9))))
    WeightMinimum=0
    Weight=@n48.Result
   )
   n49 = &RigUnit_GetRelativeTransformForItem(
    Child=(Type=Bone,Name="deform_shoulderpad_target_side_r")
    Parent=(Type=Bone,Name="upperarm_twist_02_r")
    bChildInitial=false
    bParentInitial=false
   )
   n50 = &RigVMFunction_AlphaInterp(
    InterpSpeedDecreasing=30
    InterpSpeedIncreasing=30
    Scale=1
    bClampResult=true
    Value=&RigVMFunction_MathDoubleRemap(SourceMaximum=12, SourceMinimum=-9, TargetMaximum=1, TargetMinimum=0, Value=@n49.RelativeTransform.Translation.Z).Result
   )
   > &RigUnit_ModifyTransforms(
    ItemToModify=((Item=(Type=Bone,Name="deform_shoulderpad_r"),Transform=(Rotation=(X=0,Y=-0.3827,Z=0,W=0.9239),Translation=(X=-1,Y=0,Z=-9),Scale3D=(X=0.9,Y=1,Z=0.9))))
    WeightMinimum=0
    Weight=@n50.Result
   )
   > Get Delta Ref_2_2 = RigVMFunctionReferenceNode(Item=(Type=Bone,Name="clavicle_r"))
   n51 = &RigVMFunction_AlphaInterp(
    InterpSpeedDecreasing=30
    InterpSpeedIncreasing=30
    Scale=1
    bClampResult=true
    Value=&RigVMFunction_MathDoubleRemap(SourceMaximum=30, SourceMinimum=0, TargetMaximum=1, TargetMinimum=0, Value=@"Get Delta Ref_2_2".Result.Rotation.Yaw).Result
   )
   > &RigUnit_ModifyTransforms(
    ItemToModify=((Item=(Type=Bone,Name="deform_shoulderpad_r_base_root"),Transform=(Rotation=(X=0,Y=0,Z=-0.1478,W=0.989),Translation=(X=0,Y=1,Z=0),Scale3D=(X=1,Y=1,Z=1))))
    WeightMinimum=0
    Weight=@n51.Result
   )
   n52 = &RigVMFunction_AlphaInterp(
    InterpSpeedDecreasing=30
    InterpSpeedIncreasing=30
    Scale=1
    bClampResult=true
    Value=&RigVMFunction_MathDoubleRemap(SourceMaximum=-30, SourceMinimum=0, TargetMaximum=1, TargetMinimum=0, Value=@"Get Delta Ref_2_2".Result.Rotation.Yaw).Result
   )
   > &RigUnit_ModifyTransforms(
    ItemToModify=((Item=(Type=Bone,Name="deform_shoulderpad_r_base_root"),Transform=(Rotation=(X=0,Y=0,Z=0.2164,W=0.9763),Translation=(X=-2,Y=0,Z=0),Scale3D=(X=1,Y=1,Z=1))))
    WeightMinimum=0
    Weight=@n52.Result
   )
   n53 = &RigVMFunction_AlphaInterp(
    InterpSpeedDecreasing=30
    InterpSpeedIncreasing=30
    Scale=1
    bClampResult=true
    Value=&RigVMFunction_MathDoubleRemap(SourceMaximum=30, SourceMinimum=0, TargetMaximum=1, TargetMinimum=0, Value=@"Get Delta Ref_2_2".Result.Rotation.Pitch).Result
   )
   > &RigUnit_ModifyTransforms(
    ItemToModify=((Item=(Type=Bone,Name="deform_shoulderpad_r"),Transform=(Rotation=(X=0,Y=0.0436,Z=0,W=0.999),Translation=(X=0,Y=0,Z=0),Scale3D=(X=0.95,Y=0.95,Z=0.95))))
    WeightMinimum=0
    Weight=@n53.Result
   )
   n54 = &RigUnit_GetRelativeTransformForItem(
    Child=(Type=Bone,Name="deform_shoulderpad_target_side_r")
    Parent=(Type=Bone,Name="lowerarm_twist_02_r")
    bChildInitial=False
    bParentInitial=False
   )
   n55 = &RigVMFunction_AlphaInterp(
    InterpSpeedDecreasing=30
    InterpSpeedIncreasing=30
    Scale=1
    bClampResult=true
    Value=&RigVMFunction_MathDoubleRemap(SourceMaximum=-1, SourceMinimum=10, TargetMaximum=1, TargetMinimum=0, Value=@n54.RelativeTransform.Translation.X).Result
   )
   > &RigUnit_ModifyTransforms(
    ItemToModify=((Item=(Type=Bone,Name="deform_shoulderpad_r"),Transform=(Rotation=(X=0,Y=0.0436,Z=0,W=0.999),Translation=(X=2,Y=0,Z=-2),Scale3D=(X=0.95,Y=1,Z=0.95))))
    WeightMinimum=0
    Weight=@n55.Result
   )
   n56 = &RigUnit_GetRelativeTransformForItem(Child=(Type=Bone,Name="lowerarm_r"), Parent=(Type=Bone,Name="clavicle_r"), bChildInitial=False, bParentInitial=False)
   n57 = &RigVMFunction_AlphaInterp(
    InterpSpeedDecreasing=30
    InterpSpeedIncreasing=30
    Scale=1
    bClampResult=true
    Value=&RigVMFunction_MathDoubleRemap(SourceMaximum=-25, SourceMinimum=0, TargetMaximum=1, TargetMinimum=0, Value=@n56.RelativeTransform.Translation.Y).Result
   )
   > &RigUnit_ModifyTransforms(
    ItemToModify=((Item=(Type=Bone,Name="deform_shoulderpad_r_base_root"),Transform=(Rotation=(X=0.2586,Y=-0.0113,Z=-0.0421,W=0.965),Translation=(X=0,Y=-1,Z=2.5),Scale3D=(X=1,Y=1,Z=1))))
    WeightMinimum=0
    Weight=@n57.Result
   )
   n58 = &RigVMFunction_AlphaInterp(
    InterpSpeedDecreasing=30
    InterpSpeedIncreasing=30
    Scale=1
    bClampResult=true
    Value=&RigVMFunction_MathDoubleRemap(SourceMaximum=25, SourceMinimum=0, TargetMaximum=1, TargetMinimum=0, Value=@n56.RelativeTransform.Translation.Y).Result
   )
   > &RigUnit_ModifyTransforms(
    ItemToModify=((Item=(Type=Bone,Name="deform_shoulderpad_r_base_root"),Transform=(Rotation=(X=-0.2578,Y=-0.0226,Z=0.0842,W=0.9623),Translation=(X=0,Y=0,Z=0),Scale3D=(X=1,Y=1,Z=1))))
    WeightMinimum=0
    Weight=@n58.Result
   )
   > n59 = &RigUnit_SphericalPoseReader(
    ActiveRegionScaleFactors=(PositiveWidth=1,NegativeWidth=1,PositiveHeight=1,NegativeHeight=1)
    ActiveRegionSize=0.07
    Debug=(bDrawDebug=false,bDraw2D=False,bDrawLocalAxes=False,DebugScale=25,DebugSegments=20,DebugThickness=0.25)
    DriverAxis=(X=-1,Y=0,Z=0)
    DriverItem=(Type=Bone,Name="upperarm_r")
    FalloffRegionScaleFactors=(PositiveWidth=1,NegativeWidth=1,PositiveHeight=1,NegativeHeight=1)
    FalloffSize=0.25
    FlipHeightScaling=False
    FlipWidthScaling=False
    OptionalParentItem=(Type=Bone,Name="clavicle_r")
    RotationOffset=(X=210,Y=36,Z=0)
   )
   > &RigUnit_ModifyTransforms(
    ItemToModify=((Item=(Type=Bone,Name="deform_shoulderpad_r"),Transform=(Rotation=(X=0.4241,Y=0.0026,Z=-0.0973,W=0.9004),Translation=(X=3.2797,Y=2.8983,Z=-0.0399),Scale3D=(X=1,Y=1,Z=1))))
    WeightMinimum=0
    Weight=@n59.OutputParam
   )
   > n60 = &RigUnit_SphericalPoseReader(
    ActiveRegionScaleFactors=(PositiveWidth=1,NegativeWidth=1,PositiveHeight=1,NegativeHeight=1)
    ActiveRegionSize=0.025
    Debug=(bDrawDebug=false,bDraw2D=False,bDrawLocalAxes=False,DebugScale=25,DebugSegments=20,DebugThickness=0.25)
    DriverAxis=(X=-1,Y=0,Z=0)
    DriverItem=(Type=Bone,Name="lowerarm_r")
    FalloffRegionScaleFactors=(PositiveWidth=1,NegativeWidth=1,PositiveHeight=1,NegativeHeight=1)
    FalloffSize=0.2
    FlipHeightScaling=False
    FlipWidthScaling=False
    OptionalParentItem=(Type=Bone,Name="upperarm_r")
    RotationOffset=(X=0,Y=85,Z=-10)
   )
   > n61 = &RigUnit_SphericalPoseReader(
    ActiveRegionScaleFactors=(PositiveWidth=1,NegativeWidth=1,PositiveHeight=1,NegativeHeight=1)
    ActiveRegionSize=0.15
    Debug=(bDrawDebug=false,bDraw2D=False,bDrawLocalAxes=False,DebugScale=25,DebugSegments=20,DebugThickness=0.25)
    DriverAxis=(X=-1,Y=0,Z=0)
    DriverItem=(Type=Bone,Name="upperarm_r")
    FalloffRegionScaleFactors=(PositiveWidth=1,NegativeWidth=1,PositiveHeight=1,NegativeHeight=1)
    FalloffSize=0.3
    FlipHeightScaling=False
    FlipWidthScaling=False
    OptionalParentItem=(Type=Bone,Name="clavicle_r")
    RotationOffset=(X=0,Y=145,Z=0)
   )
   n62 = &RigVMFunction_MathDoubleMul(B=1, A=&RigVMFunction_MathDoubleMul(A=@n61.OutputParam, B=@n60.OutputParam).Result)
   > &RigUnit_ParentConstraint(
    AdvancedSettings=(InterpolationType=Average,RotationOrderForFilter=XYZ)
    Child=(Type=Bone,Name="deform_shoulderpad_r_base_root")
    Filter=(TranslationFilter=(bX=false,bY=true,bZ=false),RotationFilter=(bX=false,bY=false,bZ=false),ScaleFilter=(bX=false,bY=false,bZ=false))
    Parents=((Item=(Type=Bone,Name="upperarm_r"),Weight=0.5),(Item=(Type=Bone,Name="lowerarm_r"),Weight=1))
    bMaintainOffset=true
    Weight=@n62.Result
   )
   > &RigUnit_ParentConstraint(
    AdvancedSettings=(InterpolationType=Average,RotationOrderForFilter=XYZ)
    Child=(Type=Bone,Name="deform_shoulderpad_r_base_root")
    Filter=(TranslationFilter=(bX=false,bY=false,bZ=false),RotationFilter=(bX=false,bY=true,bZ=true),ScaleFilter=(bX=false,bY=false,bZ=false))
    Parents=((Item=(Type=Bone,Name="clavicle_r"),Weight=0.8),(Item=(Type=Bone,Name="upperarm_r"),Weight=1))
    bMaintainOffset=false
    Weight=@n62.Result
   )
   > &RigUnit_ModifyTransforms(
    ItemToModify=((Item=(Type=Bone,Name="deform_shoulderpad_r_base_root"),Transform=(Rotation=(X=0.0007,Y=0.0131,Z=-0.0523,W=0.9985),Translation=(X=-2,Y=0,Z=-1.5),Scale3D=(X=1,Y=1,Z=1))))
    WeightMinimum=0
    Weight=@n62.Result
   )
   &RigUnit_ParentConstraint(
    AdvancedSettings=(InterpolationType=Average,RotationOrderForFilter=XZY)
    Child=(Type=Null,Name="upperarm_twist_02_r_null")
    ExecutePin=()
    Filter=(TranslationFilter=(bX=true,bY=true,bZ=true),RotationFilter=(bX=false,bY=false,bZ=false),ScaleFilter=(bX=false,bY=false,bZ=false))
    Parents=((Item=(Type=Bone,Name="upperarm_twist_02_r"),Weight=1))
    Weight=1
    bMaintainOffset=false
   )
   > &RigUnit_ParentConstraint(
    AdvancedSettings=(InterpolationType=Average,RotationOrderForFilter=XZY)
    Child=(Type=Null,Name="lowerarm_r_null")
    Filter=(TranslationFilter=(bX=true,bY=true,bZ=true),RotationFilter=(bX=false,bY=false,bZ=false),ScaleFilter=(bX=false,bY=false,bZ=false))
    Parents=((Item=(Type=Bone,Name="lowerarm_r"),Weight=1))
    Weight=1
    bMaintainOffset=false
   )
  ]}
 New Function_ContainedGraph {nodes [
   n1 = RigVMFunctionEntryNode()
   n2 = &RigVMFunction_Sequence(ExecuteContext=@n1.ExecuteContext)
   n3 = &RigUnit_SphericalPoseReader(
    ActiveRegionScaleFactors=(PositiveWidth=1,NegativeWidth=1,PositiveHeight=2,NegativeHeight=1)
    ActiveRegionSize=0.05
    Debug=(bDrawDebug=false,bDraw2D=false,bDrawLocalAxes=true,DebugScale=25,DebugSegments=20,DebugThickness=0.25)
    DriverAxis=(X=-1,Y=0,Z=0)
    DriverItem=(Type=Bone,Name="upperarm_r")
    FalloffRegionScaleFactors=(PositiveWidth=0.5,NegativeWidth=0.5,PositiveHeight=1,NegativeHeight=1)
    FalloffSize=0.3
    FlipWidthScaling=true
    OptionalParentItem=(Type=Bone,Name="spine_04")
    RotationOffset=(X=-50,Y=35,Z=0)
    ExecutePin=@n2.A
   )
   n4 = &RigVMFunction_AlphaInterp(
    InterpSpeedDecreasing=10
    InterpSpeedIncreasing=10
    bClampResult=False
    bInterpResult=False
    Value=@n3.OutputParam
    Scale=RigVMVariableNode() {variable=Pistol}.Value
   )
   > &RigUnit_ModifyTransforms(
    ItemToModify="((Item=(Type=Bone,Name=\"dyn_grenade_upper_mstr\"),Transform=(Rotation=(X=0,Y=0,Z=0,W=1),Translation=(X=0,Y=-1,Z=0),Scale3D=(X=1,Y=1,Z=1))),(Item=(Type=Bone,Name=\"dyn_grenade_lower\"),Transform=(Rotation=(X=0,Y=0,Z=0,W=1),Translation=(X=-2,Y=2,Z=0),Scale3D=(X ...(truncated)"
    Weight=@n4.Result
    ItemToModify.2.Transform.Rotation=&RigVMFunction_MathQuaternionFromEuler(Euler=(X=10,Y=20,Z=-40)).Result
    ItemToModify.1.Transform.Rotation=&RigVMFunction_MathQuaternionFromEuler(Euler=(X=10,Y=30,Z=-10)).Result
    ItemToModify.0.Transform.Rotation=&RigVMFunction_MathQuaternionFromEuler(Euler=(X=0,Y=0,Z=-15)).Result
   )
   > n5 = &RigUnit_SphericalPoseReader(
    ActiveRegionScaleFactors=(PositiveWidth=0.3,NegativeWidth=0.2,PositiveHeight=0.2,NegativeHeight=0.25)
    ActiveRegionSize=0.5
    Debug=(bDrawDebug=false,DebugScale=25,DebugSegments=20,DebugThickness=0.25)
    DriverAxis=(X=-1,Y=0,Z=0)
    DriverItem=(Type=Bone,Name="upperarm_r")
    FalloffRegionScaleFactors=(PositiveWidth=0.4,NegativeWidth=0.1,PositiveHeight=0.1,NegativeHeight=0.3)
    FalloffSize=0.3
    OptionalParentItem=(Type=Bone,Name="spine_05")
    RotationOffset=(X=-10,Y=50,Z=0)
   )
   n6 = &RigVMFunction_AlphaInterp(
    InterpSpeedDecreasing=12
    InterpSpeedIncreasing=15
    bClampResult=false
    Value=@n5.OutputParam
    Scale=RigVMVariableNode() {variable=Rifle}.Value
   )
   > &RigUnit_ModifyTransforms(
    ItemToModify="((Item=(Type=Bone,Name=\"dyn_lapel_r_mstr\"),Transform=(Rotation=(X=0,Y=0,Z=0,W=1),Translation=(X=0,Y=0,Z=0),Scale3D=(X=1,Y=1,Z=1))),(Item=(Type=Bone,Name=\"dyn_grenade_upper_mstr\"),Transform=(Rotation=(X=0,Y=0,Z=0,W=1),Translation=(X=0,Y=0,Z=3),Scale3D=(X=1, ...(truncated)"
    Weight=@n6.Result
    ItemToModify.2.Transform.Rotation=&RigVMFunction_MathQuaternionFromEuler(Euler=(X=-10,Y=10,Z=0)).Result
    ItemToModify.1.Transform.Rotation=&RigVMFunction_MathQuaternionFromEuler(Euler=(X=0,Y=20,Z=-10)).Result
    ItemToModify.0.Transform.Rotation=&RigVMFunction_MathQuaternionFromEuler(Euler=(X=30,Y=0,Z=0)).Result
   )
   > n7 = &RigUnit_SphericalPoseReader(
    ActiveRegionScaleFactors=(PositiveWidth=0.5,NegativeWidth=1,PositiveHeight=1.8,NegativeHeight=2)
    ActiveRegionSize=0.08
    Debug=(bDrawDebug=false,bDraw2D=false,bDrawLocalAxes=false,DebugScale=25,DebugSegments=20,DebugThickness=0.25)
    DriverAxis=(X=-1,Y=0,Z=0)
    DriverItem=(Type=Bone,Name="upperarm_r")
    FalloffRegionScaleFactors=(PositiveWidth=0.5,NegativeWidth=1,PositiveHeight=2,NegativeHeight=1)
    FalloffSize=0.45
    FlipHeightScaling=false
    FlipWidthScaling=true
    OptionalParentItem=(Type=Bone,Name="spine_04")
    RotationOffset=(X=-80,Y=40,Z=25)
   )
   > &RigUnit_ModifyTransforms(
    ItemToModify="((Item=(Type=Bone,Name=\"dyn_grenade_upper_mstr\"),Transform=(Rotation=(X=0,Y=0,Z=0,W=1),Translation=(X=1,Y=0,Z=0),Scale3D=(X=1,Y=1,Z=1))),(Item=(Type=Bone,Name=\"None\"),Transform=(Rotation=(X=0,Y=0,Z=0,W=1),Translation=(X=0,Y=0,Z=0),Scale3D=(X=1,Y=1,Z=1))),( ...(truncated)"
    ItemToModify.2.Transform.Rotation=&RigVMFunction_MathQuaternionFromEuler(Euler=(X=0,Y=25,Z=0), RotationOrder=XYZ).Result
    ItemToModify.0.Transform.Rotation=&RigVMFunction_MathQuaternionFromEuler(Euler=(X=0,Y=25,Z=0), RotationOrder=XYZ).Result
    Weight=@n7.OutputParam
   )
   > RigVMFunctionReturnNode()
  ]}
 New Function_ContainedGraph {nodes [
   RigVMFunctionEntryNode()
   > RigVMFunction_Sequence = RigVMAggregateNode()
   n1 = &RigUnit_SphericalPoseReader(
    ActiveRegionScaleFactors=(PositiveWidth=1,NegativeWidth=1,PositiveHeight=1,NegativeHeight=1)
    ActiveRegionSize=0.1
    Debug=(bDrawDebug=true,bDraw2D=False,bDrawLocalAxes=False,DebugScale=25,DebugSegments=20,DebugThickness=0.25)
    DriverAxis=(X=1,Y=0,Z=0)
    DriverItem=(Type=Bone,Name="head")
    FalloffRegionScaleFactors=(PositiveWidth=1,NegativeWidth=1,PositiveHeight=1,NegativeHeight=1)
    FalloffSize=0.25
    FlipHeightScaling=False
    FlipWidthScaling=False
    OptionalParentItem=(Type=Bone,Name="spine_05")
    RotationOffset=(X=0,Y=-45,Z=0)
    ExecutePin=@RigVMFunction_Sequence.A
   )
   > &RigUnit_ModifyTransforms(
    ItemToModify="((Item=(Type=Bone,Name=\"dyn_neck_back\"),Transform=(Rotation=(X=0,Y=0,Z=0.0698,W=0.9976),Translation=(X=-3,Y=0,Z=0),Scale3D=(X=1,Y=1,Z=1))),(Item=(Type=Bone,Name=\"dyn_neck_midback_r\"),Transform=(Rotation=(X=0,Y=0,Z=0,W=1),Translation=(X=-6,Y=0,Z=2),Scale3D= ...(truncated)"
    WeightMinimum=0
    ItemToModify.0.Transform.Rotation=&RigVMFunction_MathQuaternionFromEuler(Euler=(X=0,Y=-10,Z=0)).Result
    Weight=@n1.OutputParam
   )
   n2 = &RigUnit_SphericalPoseReader(
    ActiveRegionScaleFactors=(PositiveWidth=1,NegativeWidth=1,PositiveHeight=1,NegativeHeight=1)
    ActiveRegionSize=0.1
    Debug=(bDrawDebug=true,bDraw2D=False,bDrawLocalAxes=False,DebugScale=25,DebugSegments=20,DebugThickness=0.25)
    DriverAxis=(X=0,Y=1,Z=0)
    DriverItem=(Type=Bone,Name="head")
    FalloffRegionScaleFactors=(PositiveWidth=1,NegativeWidth=1,PositiveHeight=1,NegativeHeight=1)
    FalloffSize=0.25
    FlipHeightScaling=False
    FlipWidthScaling=False
    OptionalParentItem=(Type=Bone,Name="spine_05")
    RotationOffset=(X=180,Y=0,Z=0)
    ExecutePin=@RigVMFunction_Sequence.B
   )
   > &RigUnit_ModifyTransforms(
    ItemToModify="((Item=(Type=Bone,Name=\"base_neck_front_l\"),Transform=(Rotation=(X=0,Y=-0.0814,Z=0,W=0.9967),Translation=(X=-4,Y=0,Z=0),Scale3D=(X=0.9688,Y=0.9688,Z=0.9688))),(Item=(Type=Bone,Name=\"base_neck_mid_l\"),Transform=(Rotation=(X=0,Y=-0.0593,Z=0,W=0.9982),Transla ...(truncated)"
    WeightMinimum=0
    ItemToModify.0.Transform.Rotation=&RigVMFunction_MathQuaternionFromEuler(Euler=(X=0,Y=0,Z=-10)).Result
    ItemToModify.1.Transform.Rotation=&RigVMFunction_MathQuaternionFromEuler(Euler=(X=0,Y=0,Z=10)).Result
    Weight=@n2.OutputParam
   )
   n3 = &RigUnit_SphericalPoseReader(
    ActiveRegionScaleFactors=(PositiveWidth=1,NegativeWidth=1,PositiveHeight=2,NegativeHeight=2)
    ActiveRegionSize=0.1
    Debug=(bDrawDebug=false,bDraw2D=False,bDrawLocalAxes=False,DebugScale=25,DebugSegments=20,DebugThickness=0.25)
    DriverAxis=(X=1,Y=0,Z=0)
    DriverItem=(Type=Bone,Name="head")
    FalloffRegionScaleFactors=(PositiveWidth=1,NegativeWidth=1,PositiveHeight=1,NegativeHeight=1)
    FalloffSize=0.35
    FlipHeightScaling=False
    FlipWidthScaling=False
    OptionalParentItem=(Type=Bone,Name="spine_05")
    RotationOffset=(X=-90,Y=0,Z=20)
    ExecutePin=@RigVMFunction_Sequence.C
   )
   > &RigUnit_ModifyTransforms(
    ItemToModify="((Item=(Type=Bone,Name=\"dyn_neck_back\"),Transform=(Rotation=(X=0,Y=0,Z=0.0698,W=0.9976),Translation=(X=-8,Y=0,Z=0),Scale3D=(X=1,Y=1,Z=1))),(Item=(Type=Bone,Name=\"dyn_neck_midback_r\"),Transform=(Rotation=(X=0,Y=0,Z=0,W=1),Translation=(X=-7,Y=0,Z=0),Scale3D= ...(truncated)"
    WeightMinimum=0
    ItemToModify.0.Transform.Rotation=&RigVMFunction_MathQuaternionFromEuler(Euler=(X=0,Y=0,Z=0)).Result
    Weight=@n3.OutputParam
   )
   n4 = &RigUnit_SphericalPoseReader(
    ActiveRegionScaleFactors=(PositiveWidth=1,NegativeWidth=1,PositiveHeight=1,NegativeHeight=1)
    ActiveRegionSize=0.1
    Debug=(bDrawDebug=true,bDraw2D=False,bDrawLocalAxes=False,DebugScale=25,DebugSegments=20,DebugThickness=0.25)
    DriverAxis=(X=1,Y=0,Z=0)
    DriverItem=(Type=Bone,Name="head")
    FalloffRegionScaleFactors=(PositiveWidth=1,NegativeWidth=1,PositiveHeight=1,NegativeHeight=1)
    FalloffSize=0.25
    FlipHeightScaling=False
    FlipWidthScaling=False
    OptionalParentItem=(Type=Bone,Name="spine_05")
    RotationOffset=(X=-60,Y=0,Z=20)
    ExecutePin=@RigVMFunction_Sequence.D
   )
   > &RigUnit_ModifyTransforms(
    ItemToModify="((Item=(Type=Bone,Name=\"base_neck_mid_r\"),Transform=(Rotation=(X=0,Y=0,Z=0,W=1),Translation=(X=-2,Y=-3,Z=0),Scale3D=(X=1,Y=1,Z=1))),(Item=(Type=Bone,Name=\"dyn_neck_midback_r\"),Transform=(Rotation=(X=0,Y=0,Z=0,W=1),Translation=(X=-2,Y=0,Z=0),Scale3D=(X=1,Y= ...(truncated)"
    WeightMinimum=0
    ItemToModify.0.Transform.Rotation=&RigVMFunction_MathQuaternionFromEuler(Euler=(X=-10,Y=0,Z=-10)).Result
    ItemToModify.2.Transform.Rotation=&RigVMFunction_MathQuaternionFromEuler(Euler=(X=0,Y=-10,Z=0)).Result
    Weight=@n4.OutputParam
   )
   RigVMFunctionReturnNode(ExecuteContext=@RigVMFunction_Sequence.E)
  ]}
 RigVMFunctionLibrary {nodes [
   Grenades = RigVMCollapseNode(ExecuteContext=())
   JacketCollar = RigVMCollapseNode(ExecuteContext=())
   Set Transforms for Bone = RigVMCollapseNode(
    ExecuteContext=()
    Joint=()
    Rotate X=0
    Rotate Y=0
    Rotate Z=0
    Scale X=0
    Scale Y=0
    Scale Z=0
    Translate X=0
    Translate Y=0
    Translate Z=0
    Weight=1
   )
   ShoudlerPads = RigVMCollapseNode(ExecuteContext=())
  ]}
 RigVMFunction_Sequence_1_ContainedGraph {nodes [
   n1 = RigVMFunctionEntryNode()
   RigVMFunction_Sequence = &RigVMFunction_Sequence(ExecuteContext=@n1.ExecuteContext)
   n2 = &RigVMFunction_Sequence(ExecuteContext=@RigVMFunction_Sequence.B)
   n3 = &RigVMFunction_Sequence(ExecuteContext=@n2.B)
   RigVMFunctionReturnNode(A=@RigVMFunction_Sequence.A, B=@n2.A, C=@n3.A, D=@n3.B)
  ]}
 RigVMFunction_Sequence_1_ContainedGraph {nodes [
   n1 = RigVMFunctionEntryNode()
   RigVMFunction_Sequence = &RigVMFunction_Sequence(ExecuteContext=@n1.ExecuteContext)
   n2 = &RigVMFunction_Sequence(ExecuteContext=@RigVMFunction_Sequence.B)
   n3 = &RigVMFunction_Sequence(ExecuteContext=@n2.B)
   n4 = &RigVMFunction_Sequence(ExecuteContext=@n3.B)
   RigVMFunctionReturnNode(A=@RigVMFunction_Sequence.A, B=@n2.A, C=@n3.A, D=@n4.A, E=@n4.B)
  ]}
 RigVMModel {nodes [
   RigUnit_BeginExecution = &RigUnit_BeginExecution()
   FN_DeformRig_FNC = RigVMFunctionReferenceNode(
    DeformRig_ Curve="(deformRig_l_upperarm_up_alpha_72_#1=False,deformRig_r_upperarm_up_alpha_73_#2=False,deformRig_l_upperarm_fr_alpha_74_#3=False,deformRig_r_upperarm_fr_alpha_75_CB51B8 ...(truncated)"
    DeformRig_Bool="(upperarm_1_#4=True,elbow_3_#5=True,wrist_7_#6=True,groin_8_#7=True,glute_17_#8=True,knee_9_40C408324CA261 ...(truncated)"
    DeformRig_Debug="(upperarm_debug=False,wrist_debug=False,lat_debug=False,groinGlute_debug=False,upperarm_debug_18_#4=False,wrist_debug_20_#6=False,groinGlute_debug_27_#7=False,lat_deb ...(truncated)"
    DeformRig_Value="(upperarm_up_18_#9=(Translation=(X=0,Y=0,Z=8),Scale3D=(X=0.7,Y=1,Z=1)),upperarm_fwd_fr_19_#10=(Translation=(X=0,Y=10,Z=0),Scale3D=(X=0.5,Y=1,Z=1)),upperarm_fwd_bk_20_9973AAD84EE0A1E349BC9E85067D4A3 ...(truncated)"
    ExecuteContext=@RigUnit_BeginExecution.ExecutePin
   )
   > ShoudlerPads = RigVMFunctionReferenceNode()
   > Grenades = RigVMFunctionReferenceNode()
   > JacketCollar = RigVMFunctionReferenceNode()
  ]}
]
hierarchy [
 root {loc=<0 0 0>, rot=<0 0 0 1>}
 pelvis {parent=root, loc=<0 -0.7321 92.4218>, rot=<0.0127 0.707 -0.0127 -0.707>}
 spine_01 {parent=pelvis, loc=<5.6002 -0.5128 0>, rot=<0 0 0.034 -0.9994>}
 spine_02 {parent=spine_01, loc=<8.0756 0.0139 0>, rot=<0 0 0.0238 -0.9997>}
 spine_03 {parent=spine_02, loc=<7.0395 0 0>, rot=<0 0 0.0149 -0.9999>}
 spine_04 {parent=spine_03, loc=<8.1023 0 0>, rot=<0 0 -0.0132 -0.9999>}
 spine_05 {parent=spine_04, loc=<10.0141 0 0>, rot=<0 0 -0.0015 -1>}
 clavicle_l {parent=spine_05, loc=<2.5856 -0.2866 -4.3777>, rot=<0.0465 -0.7289 -0.009 -0.6829>}
 upperarm_l {parent=clavicle_l, loc=<18.2264 0.1342 -0.001>, rot=<0.0253 -0.2693 0.0013 -0.9627>}
 lowerarm_l {parent=upperarm_l, loc=<29.0071 0.1686 -1.239>, rot=<0.0391 -0.0099 -0.2453 -0.9686>}
 hand_l {parent=lowerarm_l, loc=<25.8607 0.222 1.0274>, rot=<-0.7193 -0.0988 0.1193 -0.6773>}
 thumb_01_l {parent=hand_l, loc=<1.3118 -0.3191 -2.459>, rot=<0.6546 -0.4042 0.0664 -0.6354>}
 thumb_02_l {parent=thumb_01_l, loc=<5.9336 -0.0448 0.0401>, rot=<-0.0037 0.0213 0.298 -0.9543>}
 thumb_03_l {parent=thumb_02_l, loc=<2.2223 0.3682 -0.0907>, rot=<0.0118 -0.013 0.0066 -0.9998>}
 index_metacarpal_l {parent=hand_l, loc=<1.6535 0.0183 -2.96>, rot=<0.0158 -0.0124 -0.0173 -0.9996>}
 index_01_l {parent=index_metacarpal_l, loc=<8.3008 0.0357 -0.0524>, rot=<-0.0044 0.0485 0.0924 -0.9945>}
 index_02_l {parent=index_01_l, loc=<3.3587 -0.3004 0.0452>, rot=<-0.002 0.0027 0.3225 -0.9466>}
 index_03_l {parent=index_02_l, loc=<2.5578 -0.123 -0.0252>, rot=<-0.002 0.0067 0.2558 -0.9667>}
 middle_metacarpal_l {parent=hand_l, loc=<2.1856 -0.1725 -0.5984>, rot=<0.006 0.0168 -0.0329 -0.9993>}
 middle_01_l {parent=middle_metacarpal_l, loc=<7.3275 -0.2447 0.0129>, rot=<-0.0208 0.0042 0.0858 -0.9961>}
 middle_02_l {parent=middle_01_l, loc=<4.0123 0.1401 0.1175>, rot=<-0.0218 0.0258 0.3512 -0.9357>}
 middle_03_l {parent=middle_02_l, loc=<2.6077 0.0534 0.1373>, rot=<0.0082 0.0188 0.1708 -0.9851>}
 ring_metacarpal_l {parent=hand_l, loc=<2.2074 -0.3625 1.3638>, rot=<-0.0284 0.0722 -0.0458 -0.9959>}
 ring_01_l {parent=ring_metacarpal_l, loc=<7.0295 -0.6197 -0.1446>, rot=<-0.0489 -0.0252 0.1892 -0.9804>}
 ring_02_l {parent=ring_01_l, loc=<3.6174 0.4647 -0.1027>, rot=<0 0 0.2806 -0.9598>}
 ring_03_l {parent=ring_02_l, loc=<2.4152 0.114 0.0199>, rot=<0 0 0.1189 -0.9929>}
 pinky_metacarpal_l {parent=hand_l, loc=<1.7616 -0.9753 3.3814>, rot=<-0.0397 0.0814 -0.069 -0.9935>}
 pinky_01_l {parent=pinky_metacarpal_l, loc=<7.1709 -0.4831 -0.1094>, rot=<-0.0654 -0.0035 0.2242 -0.9723>}
 pinky_02_l {parent=pinky_01_l, loc=<2.8274 0.0893 -0.1043>, rot=<0.0037 0.0011 0.2854 -0.9584>}
 pinky_03_l {parent=pinky_02_l, loc=<2.2046 0.0307 -0.0078>, rot=<-0.0002 0.0098 0.1445 -0.9895>}
 deform_wrist_dn_base_l {parent=hand_l, loc=<0 -3.4173 0>, rot=<0 0 0 -1>}
 deform_wrist_dn_l {parent=deform_wrist_dn_base_l, loc=<0 0 0>, rot=<0 0 0 1>}
 deform_wrist_up_base_l {parent=hand_l, loc=<0 2.7805 0>, rot=<0 0 0 -1>}
 deform_wrist_up_l {parent=deform_wrist_up_base_l, loc=<0 0 0>, rot=<0 0 0 1>}
 lowerarm_twist_02_l {parent=lowerarm_l, loc=<9.6199 0 0>, rot=<0 0 0 -1>}
 lowerarm_twist_01_l {parent=lowerarm_l, loc=<17.8549 0 0>, rot=<0 0 0 -1>}
 deform_elbow_in_base_l {parent=lowerarm_l, loc=<0.2112 5.645 0.8274>, rot=<0 0 0 -1>}
 deform_elbow_in_l {parent=deform_elbow_in_base_l, loc=<0 0 0>, rot=<0 0 0 1>}
 deform_elbow_out_base_l {parent=lowerarm_l, loc=<0 -7.5129 0>, rot=<0 0 0 -1>}
 deform_elbow_out_l {parent=deform_elbow_out_base_l, loc=<0 0 0>, rot=<0 0 0 1>}
 upperarm_twist_01_l {parent=upperarm_l, loc=<7.9796 0.0843 -0.6195>, rot=<0 0 0 -1>}
 deform_upperarm_bk_base_l {parent=upperarm_twist_01_l, loc=<-6.8983 -13.3826 -1.3258>, rot=<0 0 0 1>}
 deform_upperarm_bk_l {parent=deform_upperarm_bk_base_l, loc=<0 0 0>, rot=<0 0 0 1>}
 deform_upperarm_fr_base_l {parent=upperarm_twist_01_l, loc=<-8.3487 10.7117 -2.1455>, rot=<0 0 0 1>}
 deform_upperarm_fr_l {parent=deform_upperarm_fr_base_l, loc=<0 0 0>, rot=<0 0 0 1>}
 deform_upperarm_up_base_l {parent=upperarm_twist_01_l, loc=<-11.2012 -1.2234 8.7784>, rot=<0 0 0 1>}
 deform_upperarm_up_l {parent=deform_upperarm_up_base_l, loc=<0 0 0>, rot=<0 0 0 1>}
 upperarm_twist_02_l {parent=upperarm_l, loc=<19.5059 0.0843 -0.6195>, rot=<0 0 0 -1>}
 deform_bicep_base_l {parent=upperarm_twist_02_l, loc=<0 8.0494 0>, rot=<0 0 0 1>}
 deform_bicep_l {parent=deform_bicep_base_l, loc=<0 0 0>, rot=<0 0 0 1>}
 dyn_shoulder_armor_l {parent=upperarm_l, loc=<0 0 0>, rot=<0 0 0 -1>}
 deform_shoulderpad_l_base_root {parent=upperarm_l, loc=<0.0634 -1.0293 -0.0318>, rot=<-0.0177 0.302 -0.0245 -0.9528>}
 deform_shoulderpad_l_base {parent=deform_shoulderpad_l_base_root, loc=<1.4726 0 3.546>, rot=<0 0 0 1>}
 deform_shoulderpad_l {parent=deform_shoulderpad_l_base, loc=<0 0 0>, rot=<0 0 0 1>}
 deform_shoulderpad_target_up_l {parent=deform_shoulderpad_l, loc=<-2.381 -1.1601 15.217>, rot=<0 0 0 1>}
 deform_shoulderpad_target_side_l {parent=deform_shoulderpad_l, loc=<14.0512 -1.4386 -3.872>, rot=<0 0 0 1>}
 dyn_shoulder_MSTR {parent=upperarm_l, loc=<1.1187 -0.9098 5.6424>, rot=<0.4628 -0.5556 0.5382 0.4329>}
 dyn_shoulder_01 {parent=dyn_shoulder_MSTR, loc=<0.0202 0 0>, rot=<0 0 0 1>}
 neck_01 {parent=spine_05, loc=<10.815 1.9196 0>, rot=<0 0 -0.1864 -0.9825>}
 neck_02 {parent=neck_01, loc=<4.7347 0.3045 0>, rot=<0 0 0.0363 -0.9993>}
 head {parent=neck_02, loc=<5.7632 0.5552 0>, rot=<0 0 0.115 -0.9934>}
 base_neck_mid_l {parent=spine_05, loc=<13.9111 -0.3168 -14>, rot=<-0.5517 -0.0903 -0.0134 -0.8291>}
 dyn_neck_mid_l {parent=base_neck_mid_l, loc=<0 0 0>, rot=<0 0 0 1>}
 base_neck_front_l {parent=spine_05, loc=<5.972 11.0837 -13>, rot=<-0.562 -0.4255 -0.1383 -0.6957>}
 dyn_neck_front_l {parent=base_neck_front_l, loc=<0 0 0>, rot=<0 0 0 1>}
 dyn_neck_back {parent=spine_05, loc=<17.7012 -10.3662 -0.0054>, rot=<0 0 -0.147 -0.9891>}
 base_neck_front_r {parent=spine_05, loc=<5.972 11.0837 13>, rot=<0.581 0.4767 -0.1803 -0.6346>}
 dyn_neck_front_r {parent=base_neck_front_r, loc=<0 0 0>, rot=<0 0 0 1>}
 base_neck_mid_r {parent=spine_05, loc=<13.9111 -0.3168 14>, rot=<0.5632 0.0627 -0.0217 -0.8237>}
 dyn_neck_mid_r {parent=base_neck_mid_r, loc=<0 0 0>, rot=<0 0 0 1>}
 base_neck_midback_r {parent=spine_05, loc=<16.4024 -9.0642 8.6039>, rot=<0.2536 0.0238 -0.0513 -0.9657>}
 dyn_neck_midback_r {parent=base_neck_midback_r, loc=<0 0 0>, rot=<0 0 0 1>}
 clavicle_r {parent=spine_05, loc=<2.5855 -0.2866 4.3777>, rot=<0.7289 0.0465 -0.6829 0.009>}
 upperarm_r {parent=clavicle_r, loc=<-18.2264 -0.1342 0.001>, rot=<0.0253 -0.2693 0.0013 -0.9627>}
 lowerarm_r {parent=upperarm_r, loc=<-29.0071 -0.1686 1.239>, rot=<0.0391 -0.0099 -0.2453 -0.9686>}
 hand_r {parent=lowerarm_r, loc=<-25.8607 -0.222 -1.0274>, rot=<-0.7193 -0.0988 0.1193 -0.6773>}
 thumb_01_r {parent=hand_r, loc=<-1.3118 0.3191 2.459>, rot=<0.6546 -0.4042 0.0664 -0.6354>}
 thumb_02_r {parent=thumb_01_r, loc=<-5.9336 0.0448 -0.0401>, rot=<-0.0037 0.0213 0.298 -0.9543>}
 thumb_03_r {parent=thumb_02_r, loc=<-2.2223 -0.3682 0.0907>, rot=<0.0118 -0.013 0.0066 -0.9998>}
 index_metacarpal_r {parent=hand_r, loc=<-1.6535 -0.0183 2.96>, rot=<0.0158 -0.0124 -0.0173 -0.9996>}
 index_01_r {parent=index_metacarpal_r, loc=<-8.3008 -0.0357 0.0524>, rot=<-0.0044 0.0485 0.0924 -0.9945>}
 index_02_r {parent=index_01_r, loc=<-3.3587 0.3004 -0.0452>, rot=<-0.002 0.0027 0.3225 -0.9466>}
 index_03_r {parent=index_02_r, loc=<-2.5578 0.123 0.0252>, rot=<-0.002 0.0067 0.2558 -0.9667>}
 middle_metacarpal_r {parent=hand_r, loc=<-2.1856 0.1725 0.5984>, rot=<0.006 0.0168 -0.0329 -0.9993>}
 middle_01_r {parent=middle_metacarpal_r, loc=<-7.3275 0.2447 -0.0129>, rot=<-0.0208 0.0042 0.0858 -0.9961>}
 middle_02_r {parent=middle_01_r, loc=<-4.0123 -0.1401 -0.1175>, rot=<-0.0218 0.0258 0.3512 -0.9357>}
 middle_03_r {parent=middle_02_r, loc=<-2.6077 -0.0534 -0.1373>, rot=<0.0082 0.0188 0.1708 -0.9851>}
 ring_metacarpal_r {parent=hand_r, loc=<-2.2074 0.3625 -1.3638>, rot=<-0.0284 0.0722 -0.0458 -0.9959>}
 ring_01_r {parent=ring_metacarpal_r, loc=<-7.0295 0.6197 0.1446>, rot=<-0.0489 -0.0252 0.1892 -0.9804>}
 ring_02_r {parent=ring_01_r, loc=<-3.6174 -0.4647 0.1027>, rot=<0 0 0.2806 -0.9598>}
 ring_03_r {parent=ring_02_r, loc=<-2.4152 -0.114 -0.0199>, rot=<0 0 0.1189 -0.9929>}
 pinky_metacarpal_r {parent=hand_r, loc=<-1.7616 0.9753 -3.3814>, rot=<-0.0397 0.0814 -0.069 -0.9935>}
 pinky_01_r {parent=pinky_metacarpal_r, loc=<-7.1709 0.4831 0.1094>, rot=<-0.0654 -0.0035 0.2242 -0.9723>}
 pinky_02_r {parent=pinky_01_r, loc=<-2.8274 -0.0893 0.1043>, rot=<0.0037 0.0011 0.2854 -0.9584>}
 pinky_03_r {parent=pinky_02_r, loc=<-2.2046 -0.0307 0.0078>, rot=<-0.0002 0.0098 0.1445 -0.9895>}
 deform_wrist_dn_base_r {parent=hand_r, loc=<0 3.4172 0>, rot=<0 0 0 -1>}
 deform_wrist_dn_r {parent=deform_wrist_dn_base_r, loc=<0 0 0>, rot=<0 0 0 -1>}
 deform_wrist_up_base_r {parent=hand_r, loc=<-0.0001 -2.7804 0>, rot=<0 0 0 -1>}
 deform_wrist_up_r {parent=deform_wrist_up_base_r, loc=<0 0 0>, rot=<0 0 0 -1>}
 lowerarm_twist_02_r {parent=lowerarm_r, loc=<-9.6199 0 0>, rot=<0 0 0 -1>}
 lowerarm_twist_01_r {parent=lowerarm_r, loc=<-17.8549 0 0>, rot=<0 0 0 -1>}
 deform_elbow_in_base_r {parent=lowerarm_r, loc=<-0.2114 -5.6449 -0.827>, rot=<0 0 0 -1>}
 deform_elbow_in_r {parent=deform_elbow_in_base_r, loc=<0 0 0>, rot=<0 0 0 1>}
 deform_elbow_out_base_r {parent=lowerarm_r, loc=<-0.0001 7.5129 0.0002>, rot=<0 0 0 -1>}
 deform_elbow_out_r {parent=deform_elbow_out_base_r, loc=<0 0 0>, rot=<0 0 0 1>}
 upperarm_twist_01_r {parent=upperarm_r, loc=<-7.9796 -0.0843 0.6195>, rot=<0 0 0 -1>}
 deform_upperarm_bk_base_r {parent=upperarm_twist_01_r, loc=<6.8985 13.3826 1.3257>, rot=<0 0 0 1>}
 deform_upperarm_bk_r {parent=deform_upperarm_bk_base_r, loc=<0 0 0>, rot=<0 0 0 1>}
 deform_upperarm_fr_base_r {parent=upperarm_twist_01_r, loc=<8.3485 -10.7117 2.1456>, rot=<0 0 0 1>}
 deform_upperarm_fr_r {parent=deform_upperarm_fr_base_r, loc=<0 0 0>, rot=<0 0 0 1>}
 deform_upperarm_up_base_r {parent=upperarm_twist_01_r, loc=<11.2015 1.2234 -8.7788>, rot=<0 0 0 1>}
 deform_upperarm_up_r {parent=deform_upperarm_up_base_r, loc=<0 0 0>, rot=<0 0 0 1>}
 upperarm_twist_02_r {parent=upperarm_r, loc=<-19.5059 -0.0843 0.6195>, rot=<0 0 0 -1>}
 deform_bicep_base_r {parent=upperarm_twist_02_r, loc=<0.0002 -8.0494 -0.0003>, rot=<0 0 0 1>}
 deform_bicep_r {parent=deform_bicep_base_r, loc=<0 0 0>, rot=<0 0 0 1>}
 dyn_shoulder_armor_r {parent=upperarm_r, loc=<0 0 -0.0001>, rot=<0 0 0 -1>}
 deform_shoulderpad_r_base_root {parent=upperarm_r, loc=<-0.0634 1.0293 0.0317>, rot=<-0.0177 0.302 -0.0245 -0.9528>}
 deform_shoulderpad_r_base {parent=deform_shoulderpad_r_base_root, loc=<-1.4726 0 -3.5455>, rot=<0 0 0 1>}
 deform_shoulderpad_r {parent=deform_shoulderpad_r_base, loc=<0 0 0>, rot=<0 0 0 1>}
 deform_shoulderpad_target_up_r {parent=deform_shoulderpad_r, loc=<2.381 1.1601 -15.2174>, rot=<0 0 0 1>}
 deform_shoulderpad_target_side_r {parent=deform_shoulderpad_r, loc=<-14.0512 1.4386 3.8713>, rot=<0 0 0 1>}
 dyn_shoulder_r_MSTR {parent=upperarm_r, loc=<-1.1188 0.9098 -5.6421>, rot=<-0.5556 -0.4628 0.4329 -0.5382>}
 dyn_shoulder_r_01 {parent=dyn_shoulder_r_MSTR, loc=<0.0207 0.0002 0>, rot=<0 0 0 1>}
 attach_cape {parent=spine_05, loc=<9.017 -11.714 0>, rot=<0.0616 -0.7044 -0.0616 -0.7044>}
 attach_backpack {parent=spine_05, loc=<-0.123 -13.473 0>, rot=<-0.0093 -0.707 0.0093 -0.707>}
 dyn_necklace_mstr {parent=spine_05, loc=<8.205 8.7479 0.0653>, rot=<0.1825 0.6711 0.6915 0.1952>}
 dyn_necklace {parent=dyn_necklace_mstr, loc=<0 0 0>, rot=<0 0 0 1>}
 dyn_necklace_charm {parent=dyn_necklace, loc=<5.9896 0.0174 1.2407>, rot=<0 0 0 1>}
 deform_pec_aim_base_l {parent=spine_04, loc=<5.0227 13.6719 -2.6508>, rot=<0.0672 -0.5201 0.3228 -0.7879>}
 deform_pec_aim_l {parent=deform_pec_aim_base_l, loc=<0 0 0>, rot=<0 0 0 1>}
 deform_pec_base_l {parent=deform_pec_aim_l, loc=<16.3279 0 0>, rot=<0 0 0 1>}
 deform_pec_l {parent=deform_pec_base_l, loc=<0 0 0>, rot=<0 0 0 1>}
 deform_pec_aim_base_r {parent=spine_04, loc=<5.0225 13.6718 2.6508>, rot=<-0.5201 -0.0672 0.7879 0.3228>}
 deform_pec_aim_r {parent=deform_pec_aim_base_r, loc=<0 0 0>, rot=<0 0 0 1>}
 deform_pec_base_r {parent=deform_pec_aim_r, loc=<-16.3279 0 0>, rot=<0 0 0 -1>}
 deform_pec_r {parent=deform_pec_base_r, loc=<0 0 0.0002>, rot=<0 0 0 -1>}
 dyn_lapel_r_mstr {parent=spine_04, loc=<3.2161 18.8553 11.9017>, rot=<0.6993 0.0688 0.6916 0.167>}
 dyn_lapel_r {parent=dyn_lapel_r_mstr, loc=<0 0 0>, rot=<0 0 0 1>}
 dyn_grenade_lower {parent=dyn_lapel_r, loc=<3.8409 -0.3249 -3.9051>, rot=<-0.2181 -0.3047 -0.1435 -0.916>}
 dyn_lapel_l_mstr {parent=spine_04, loc=<4.1053 20.5841 -11.9017>, rot=<-0.167 -0.6916 -0.0688 -0.6993>}
 dyn_lapel_l {parent=dyn_lapel_l_mstr, loc=<0 0 0>, rot=<0 0 0 1>}
 dyn_grenade_upper_mstr {parent=spine_04, loc=<7.9724 12.9942 16.3541>, rot=<0.6238 0.2429 0.735 0.1083>}
 dyn_grenade_upper {parent=dyn_grenade_upper_mstr, loc=<0 0 0>, rot=<0 0 0 1>}
 deform_lat_l_base {parent=spine_03, loc=<2.5561 -10.3716 -17.6918>, rot=<0 0 0 -1>}
 deform_lat_l {parent=deform_lat_l_base, loc=<0 0 0>, rot=<0 0 0 1>}
 deform_lat_r_base {parent=spine_03, loc=<2.5562 -10.3716 17.6918>, rot=<0 0 1 0>}
 deform_lat_r {parent=deform_lat_r_base, loc=<0 0 0>, rot=<0 0 0 -1>}
 dyn_jacket_l_01 {parent=spine_03, loc=<-0.973 16.8461 -17>, rot=<-0.0111 -0.9392 0.3378 -0.0606>}
 dyn_jacket_l_02 {parent=dyn_jacket_l_01, loc=<6.9239 0 0>, rot=<0 0 0 1>}
 dyn_jacket_r_01 {parent=spine_03, loc=<-0.973 16.8461 17>, rot=<-0.0464 -0.9031 -0.4257 0.0331>}
 dyn_jacket_r_02 {parent=dyn_jacket_r_01, loc=<6.923 0 0.0001>, rot=<0 0 0 1>}
 dyn_jacket_back_01 {parent=spine_03, loc=<4.1533 -14.9761 0>, rot=<0 0 1 0.0036>}
 dyn_jacket_back_02 {parent=dyn_jacket_back_01, loc=<6.8639 0 0>, rot=<0 0 0 1>}
 dyn_belly_mstr {parent=spine_01, loc=<0 9 0>, rot=<0.4919 0.508 0.508 0.4919>}
 dyn_belly {parent=dyn_belly_mstr, loc=<0 0 0>, rot=<0 0 0 1>}
 thigh_l {parent=pelvis, loc=<-1.8716 -0.0109 -11.9936>, rot=<0.0012 0.0572 0.0155 -0.9982>}
 calf_l {parent=thigh_l, loc=<-41.9178 -0.2246 0.0926>, rot=<0.0006 -0.0161 -0.0391 -0.9991>}
 foot_l {parent=calf_l, loc=<-38.7057 1.2438 0.2452>, rot=<0.0975 -0.0474 0.0317 -0.9936>}
 ball_l {parent=foot_l, loc=<-6.1614 13.1469 0.6807>, rot=<0.0001 -0.0001 -0.7063 -0.7079>}
 calf_twist_02_l {parent=calf_l, loc=<-15.9268 -0.0967 0.0397>, rot=<0.0237 0.0095 -0.0035 -0.9997>}
 calf_twist_01_l {parent=calf_l, loc=<-32.9335 -0.0967 0.0397>, rot=<0.0237 0.0095 -0.0035 -0.9997>}
 deform_knee_in_base_l {parent=calf_l, loc=<1.298 -8.0356 -0.0496>, rot=<0 0 0 -1>}
 deform_knee_in_l {parent=deform_knee_in_base_l, loc=<0 0 0>, rot=<0 0 0 1>}
 deform_knee_out_base_l {parent=calf_l, loc=<0 6.8739 0>, rot=<0 0 0 -1>}
 deform_knee_out_l {parent=deform_knee_out_base_l, loc=<0 0 0>, rot=<0 0 0 1>}
 thigh_twist_01_l {parent=thigh_l, loc=<-21.6303 0 0>, rot=<0.0013 0 0 -1>}
 deform_groin_base_l {parent=thigh_l, loc=<-2.1908 8.5661 -1.7442>, rot=<0 0 0 -1>}
 deform_groin_l {parent=deform_groin_base_l, loc=<0 0 0>, rot=<0 0 0 1>}
 dyn_chain_01 {parent=thigh_l, loc=<-5.0014 14.0439 -3.2776>, rot=<0.1396 -0.9224 -0.2953 -0.2063>}
 dyn_chain_02 {parent=dyn_chain_01, loc=<9.1575 -1.4535 -1.3625>, rot=<-0.03 -0.2456 0.0793 -0.9657>}
 dyn_chain_03 {parent=dyn_chain_02, loc=<6.4616 -1.0169 0.7977>, rot=<-0.0765 -0.1042 0.2748 -0.9528>}
 dyn_chain_05 {parent=thigh_l, loc=<-4.197 -13.2993 -0.2364>, rot=<0.3238 0.1182 0.177 -0.9219>}
 dyn_chain_06 {parent=dyn_chain_05, loc=<-7.296 0.2917 0.4051>, rot=<0.0069 0.0252 0.1153 -0.993>}
 dyn_chain_07 {parent=dyn_chain_06, loc=<-7.8713 0.3506 1.2639>, rot=<-0.0019 -0.0353 0.0505 -0.9981>}
 dyn_chain_c {parent=thigh_l, loc=<-19.9733 0.3395 -8.8553>, rot=<0.0616 -0.7044 -0.0616 -0.7044>}
 thigh_r {parent=pelvis, loc=<-1.8716 -0.0109 11.9936>, rot=<0.0572 -0.0012 0.9982 0.0155>}
 calf_r {parent=thigh_r, loc=<41.9178 0.2246 -0.0926>, rot=<0.0006 -0.0161 -0.0391 -0.9991>}
 foot_r {parent=calf_r, loc=<38.7057 -1.2438 -0.2452>, rot=<0.0975 -0.0474 0.0317 -0.9936>}
 ball_r {parent=foot_r, loc=<6.1614 -13.1469 -0.6807>, rot=<0.0001 -0.0001 -0.7063 -0.7079>}
 calf_twist_02_r {parent=calf_r, loc=<15.9268 0.0967 -0.0397>, rot=<0.0237 0.0095 -0.0035 -0.9997>}
 calf_twist_01_r {parent=calf_r, loc=<32.9335 0.0967 -0.0397>, rot=<0.0237 0.0095 -0.0035 -0.9997>}
 deform_knee_in_base_r {parent=calf_r, loc=<-1.2981 8.0356 0.0496>, rot=<0 0 0 -1>}
 deform_knee_in_r {parent=deform_knee_in_base_r, loc=<0 0 0>, rot=<0 0 0 1>}
 deform_knee_out_base_r {parent=calf_r, loc=<0 -6.8739 0>, rot=<0 0 0 -1>}
 deform_knee_out_r {parent=deform_knee_out_base_r, loc=<0 0 0>, rot=<0 0 0 1>}
 thigh_twist_01_r {parent=thigh_r, loc=<21.6303 0 0>, rot=<0.0013 0 0 -1>}
 deform_groin_base_r {parent=thigh_r, loc=<2.1908 -8.5661 1.7442>, rot=<0 0 0 -1>}
 deform_groin_r {parent=deform_groin_base_r, loc=<0 0 0>, rot=<0 0 0 -1>}
 deform_glute_base_l {parent=pelvis, loc=<-10.1877 -11.6217 -12.899>, rot=<0.0275 0.0501 -0.4472 -0.8926>}
 deform_glute_l {parent=deform_glute_base_l, loc=<0 0 0>, rot=<0 0 0 1>}
 deform_glute_base_r {parent=pelvis, loc=<-10.1877 -11.6218 12.899>, rot=<0.0501 -0.0275 0.8926 -0.4472>}
 deform_glute_r {parent=deform_glute_base_r, loc=<0 0 0>, rot=<0 0 0 1>}
]
variables [
 Crouch {type=double, default=0}
 Emote {type=bool, default=False}
 EmoteAlpha {type=double, default=0}
 Fly_mode {type=double, default=0}
 Parachute {type=double, default=0}
 Pistol {type=double, default=0}
 Rifle {type=double, default=0}
 SkyDive {type=double, default=0}
 Targeting {type=double, default=0}
]
