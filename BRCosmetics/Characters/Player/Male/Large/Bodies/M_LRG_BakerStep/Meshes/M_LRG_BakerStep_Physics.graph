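# Physics asset for the M_LRG_BakerStep large male player body cosmetic, covering dynamic bones around the neck, likely for secondary motion.
# Five dyn_neck capsules (front and mid on both sides, midback on the right only) are each constrained to a tiny matching base_neck capsule.
# Limits allow a wider first swing axis with tight second swing and twist.
# Large spheres on both upper arms supply shoulder-area collision shapes.
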PhysicsAsset /BRCosmetics/Characters/Player/Male/Large/Bodies/M_LRG_BakerStep/Meshes/M_LRG_BakerStep_Physics
bodies {
 base_neck_front_l [{shapeType=Capsule, length=0.505, radius=0.505, shapeName=base_neck_front_l_capsule, rotation=<rot 0 -90 0>}]
 base_neck_front_r [{shapeType=Capsule, length=0.505, radius=0.505, shapeName=base_neck_front_r_capsule, rotation=<rot 0 -90 0>}]
 base_neck_mid_l [{shapeType=Capsule, length=0.505, radius=0.505, shapeName=base_neck_mid_l_capsule, rotation=<rot 0 -90 0>}]
 base_neck_mid_r [{shapeType=Capsule, length=0.505, radius=0.505, shapeName=base_neck_mid_r_capsule, rotation=<rot 0 -90 0>}]
 base_neck_midback_r [{shapeType=Capsule, length=0.505, radius=0.505, shapeName=base_neck_midback_r_capsule, rotation=<rot 0 -90 0>}]
 dyn_neck_front_l [{shapeType=Capsule, length=9.1948, radius=2.7259, shapeName=dyn_neck_front_r_capsule, center=<4.4204 0 0>}]
 dyn_neck_front_r [{shapeType=Capsule, length=8.6948, radius=3.2259, shapeName=dyn_neck_front_r_capsule, center=<4.4204 0 0>}]
 dyn_neck_mid_l [{shapeType=Capsule, length=8.6541, radius=2.9361, shapeName=dyn_neck_mid_r_capsule, center=<4.4204 0 0>}]
 dyn_neck_mid_r [{shapeType=Capsule, length=8.6541, radius=2.9361, shapeName=dyn_neck_mid_r_capsule, center=<4.4204 0 0>}]
 dyn_neck_midback_r [{shapeType=Capsule, length=2.4896, radius=1.4433, shapeName=dyn_neck_midback_r_capsule, center=<4.4204 0 0>}]
 upperarm_l [{shapeType=Sphere, radius=11.3421, shapeName=upperarm_l_sphere, center=<2 -1.4 0>}]
 upperarm_r [{shapeType=Sphere, radius=11.8467, shapeName=upperarm_r_sphere, center=<-2 1.4 0>}]
}
constraints [
 {bone1Name=dyn_neck_front_l, bone2Name=base_neck_front_l, swing1LimitDegrees=20, swing2LimitDegrees=5, twistLimitDegrees=5}
 {bone1Name=dyn_neck_front_r, bone2Name=base_neck_front_r, swing1LimitDegrees=20, swing2LimitDegrees=5, twistLimitDegrees=5}
 {bone1Name=dyn_neck_mid_l, bone2Name=base_neck_mid_l, swing1LimitDegrees=25, swing2LimitDegrees=5, twistLimitDegrees=5}
 {bone1Name=dyn_neck_mid_r, bone2Name=base_neck_mid_r, swing1LimitDegrees=25, swing2LimitDegrees=5, twistLimitDegrees=5}
 {bone1Name=dyn_neck_midback_r, bone2Name=base_neck_midback_r, swing1LimitDegrees=20, swing2LimitDegrees=5, twistLimitDegrees=5}
]
