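# Secondary-motion AnimBP for the large male BakerStep character body, animating chains, grenades, jacket, lapels and necklace.
# Pose copied from the parent mesh runs through RigidBody, weighted constraints, AnimDynamics, additive bone tweaks and a Control Rig.
# Event graph turns movement, weapon, targeting, emote and front-end state into alphas and rig inputs.
# Lobby idle curve layers DefaultSlot from dyn_shoulder_MSTR; belly and jacket_back_02 dynamics are disconnected from output.
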AnimBlueprint /BRCosmetics/Characters/Player/Male/Large/Bodies/M_LRG_BakerStep/Meshes/M_LRG_BakerStep_AnimBP
target_skeleton=&/M_LRG_BakerStep_Skeleton
graphs [
 AnimGraph {nodes [
   Anim Dynamics - Bone: dyn_belly(Alpha=1, ExternalForce="0,0,0") {
    angularDampingOverride=0.9999
    angularSpringConstant=2000
    bAngularSpring=true
    bOverrideAngularDamping=true
    bUseGravityOverride=true
    simulationSpace=BoneRelative
    alphaBoolBlend {blendOption=Linear}
    boundBone {boneName=dyn_belly}
    componentAppliedLinearAccClamp=<10000 10000 10000>
    gravityOverride {z=-980}
    physicsBodyDefinitions [{
      boundBone {boneName=dyn_belly}
      boxExtents=<13 13 13>
      constraintSetup {angularLimitsMax=<10 10 20>, angularLimitsMin=<-10 -10 -5>, angularTarget=<1 0 0>}
      localJointOffset=<2 0 0>
     }]
    relativeSpaceBone {boneName=root}
    simSpaceSettings {maxAngularAcceleration=20, maxAngularVelocity=770, simSpaceAngularAlpha=1}
   }
   Anim Dynamics - Bone: dyn_jacket_back_02!(ComponentPose, bAlphaBoolEnabled:bool=True, ExternalForce="0,0,0") -> (Pose) {
    alphaInputType=Bool
    angularDampingOverride=1
    angularSpringConstant=300
    bAngularSpring=true
    bOverrideAngularDamping=true
    bUseGravityOverride=true
    simulationSpace=BoneRelative
    alphaBoolBlend {blendInTime=0.1, blendOption=Linear, blendOutTime=0.1}
    boundBone {boneName=dyn_jacket_back_02}
    componentAppliedLinearAccClamp=<3000 3000 3000>
    componentLinearAccScale=<0.4 0.4 0.4>
    componentLinearVelScale=<0.4 0.4 0.4>
    gravityOverride {z=-980}
    physicsBodyDefinitions [{
      boundBone {boneName=dyn_jacket_back_02}
      boxExtents=<10 10 10>
      constraintSetup {angularLimitsMax=<0 5 60>, angularLimitsMin=<0 -5 0>, angularTarget=<0 0 0>}
      localJointOffset=<1 0 0>
     }]
    relativeSpaceBone {boneName=pelvis}
   }
   n1 = Anim Dynamics - Bone: dyn_belly(Alpha=0.9, ExternalForce="0,0,0") {
    angularDampingOverride=0.9
    angularSpringConstant=5000
    bLinearSpring=true
    bOverrideLinearDamping=true
    bUseGravityOverride=true
    gravityScale=0.35
    linearDampingOverride=0.95
    linearSpringConstant=150
    simulationSpace=RootRelative
    boundBone {boneName=dyn_belly}
    componentAppliedLinearAccClamp=<0 0 0>
    gravityOverride {z=-980}
    physicsBodyDefinitions [{
      boundBone {boneName=dyn_belly}
      boxExtents=<35 35 35>
      constraintSetup {angularLimitsMax=<0 0 0>, angularLimitsMin=<0 0 0>, angularTarget=<1 0 0>, linearAxesMax=<1 1 1>, linearAxesMin=<0 -1 -1>}
      localJointOffset=<3 0 0>
     }]
    relativeSpaceBone {boneName=pelvis}
    simSpaceSettings {maxAngularAcceleration=15, maxAngularVelocity=100, simSpaceAngularAlpha=1}
   }
   n2 = Anim Dynamics - Bone: dyn_belly(ComponentPose=@n1, Alpha=0.9, ExternalForce="0,0,0") {
    angularDampingOverride=0.9999
    angularSpringConstant=25000
    bAngularSpring=true
    bOverrideAngularDamping=true
    bUseGravityOverride=true
    gravityScale=0.35
    simulationSpace=RootRelative
    boundBone {boneName=dyn_belly}
    componentAppliedLinearAccClamp=<0 0 0>
    gravityOverride {z=-980}
    physicsBodyDefinitions [{
      boundBone {boneName=dyn_belly}
      boxExtents=<35 35 35>
      constraintSetup {angularLimitsMax=<15 15 15>, angularLimitsMin=<-15 -15 -15>, angularTarget=<1 0 0>}
      localJointOffset=<3 0 0>
     }]
    relativeSpaceBone {boneName=pelvis}
    simSpaceSettings {maxAngularAcceleration=15, maxAngularVelocity=100, simSpaceAngularAlpha=1}
   }
   "Transform (Modify) Bone - Bone: dyn_belly"(ComponentPose=@n2, Alpha=1, Translation="0,0,5", Rotation="0,0,0", Scale="1,1,1") {
    rotationMode=BMM_Additive
    translationMode=BMM_Additive
    alphaBoolBlend {blendOption=Linear}
    alphaScaleBiasClamp {bInterpResult=true, interpSpeedDecreasing=7, interpSpeedIncreasing=5}
    boneToModify {boneName=dyn_belly}
    translation {z=5}
   }
   n3 = RigidBody!(
    ComponentPose=Local To Component(LocalPose=Copy Pose From Mesh() {bUseAttachedParent=true})
    Alpha
    OverrideWorldGravity:Vector=Get GravityOverride()
    ExternalForce
   ) -> (Pose) {
    bClampLinearTranslationLimitToRefPose=true
    bOverrideWorldGravity=true
    overridePhysicsAsset=&/M_LRG_BakerStep_Physics
    simulationSpace=BaseBoneSpace
    alphaBoolBlend {blendOption=Linear}
    baseBoneRef {boneName=spine_05}
    componentAppliedLinearAccClamp=<3000 3000 3000>
    componentLinearAccScale=<0.6 0.6 0.6>
    componentLinearVelScale=<0.6 0.6 0.6>
    simSpaceSettings {
     maxAngularAcceleration=1000
     maxAngularVelocity=3
     maxLinearAcceleration=700
     maxLinearVelocity=500
     velocityScaleZ=0
     worldAlpha=0.2
     externalLinearDragV=<0.2 0.2>
    }
   }
   n4 = "Constraint - dyn_chain_01 (PPPPPP)"!(
    ComponentPose=@n3
    Alpha=1
    ConstraintWeights_0=0.8646
    ConstraintWeights_1:float=0.002
    ConstraintWeights_2:float=0.0373
    ConstraintWeights_3:float=0.0005
    ConstraintWeights_4:float=0.0954
    ConstraintWeights_5:float=0.0002
   ) -> (Pose) {
    alphaBoolBlend {blendOption=Linear}
    boneToModify {boneName=dyn_chain_01}
    constraintSetup [
     {targetBone {boneName=thigh_twist_01_l}}
     {targetBone {boneName=spine_01}}
     {targetBone {boneName=pelvis}}
     {targetBone {boneName=thigh_twist_01_r}}
     {targetBone {boneName=thigh_l}}
     {targetBone {boneName=calf_l}}
    ]
    constraintWeights [1,1,1,1,1,1]
   }
   n5 = "Constraint - dyn_chain_05 (PPPPPPPPP)"!(
    ComponentPose=@n4
    Alpha=1
    ConstraintWeights_0=0.2988
    ConstraintWeights_1:float=0.0182
    ConstraintWeights_2:float=0.6498
    ConstraintWeights_3:float=0.0257
    ConstraintWeights_4:float=0.0019
    ConstraintWeights_5:float=0.0045
    ConstraintWeights_6:float=0.0003
    ConstraintWeights_7:float=0.0006
    ConstraintWeights_8:float=0.0002
   ) -> (Pose) {
    alphaBoolBlend {blendOption=Linear}
    boneToModify {boneName=dyn_chain_05}
    constraintSetup [
     {targetBone {boneName=pelvis}}
     {targetBone {boneName=spine_01}}
     {targetBone {boneName=thigh_l}}
     {targetBone {boneName=thigh_twist_01_l}}
     {targetBone {boneName=calf_l}}
     {targetBone {boneName=thigh_twist_01_r}}
     {targetBone {boneName=upperarm_twist_02_l}}
     {targetBone {boneName=lowerarm_twist_02_l}}
     {targetBone {boneName=thumb_03_l}}
    ]
    constraintWeights [1,1,1,1,1,1,1,1,1]
   }
   n6 = "Constraint - dyn_chain_07 (PPPPPP)"!(
    ComponentPose=@n5
    Alpha=1
    ConstraintWeights_0=0.0857
    ConstraintWeights_1:float=0.0055
    ConstraintWeights_2:float=0.0655
    ConstraintWeights_3:float=0.004
    ConstraintWeights_4:float=0.8362
    ConstraintWeights_5:float=0.0031
   ) -> (Pose) {
    alphaBoolBlend {blendOption=Linear}
    boneToModify {boneName=dyn_chain_07}
    constraintSetup [
     {targetBone {boneName=thigh_twist_01_l}}
     {targetBone {boneName=spine_01}}
     {targetBone {boneName=pelvis}}
     {targetBone {boneName=thigh_twist_01_r}}
     {targetBone {boneName=thigh_l}}
     {targetBone {boneName=calf_l}}
    ]
    constraintWeights [1,1,1,1,1,1]
   }
   n7 = "Constraint - dyn_chain_05 (PPPPPP)"!(
    ComponentPose=@n6
    Alpha=Get FrontEnd Idle Alpha() -> (FrontEnd Idle Alpha:double)
    ConstraintWeights_0=0.6374
    ConstraintWeights_1:float=0.1852
    ConstraintWeights_2:float=0.1738
    ConstraintWeights_3:float=0.0017
    ConstraintWeights_4:float=0.0015
    ConstraintWeights_5:float=0.0004
   ) -> (Pose) {
    alphaBoolBlend {blendOption=Linear}
    boneToModify {boneName=dyn_chain_05}
    constraintSetup [
     {targetBone {boneName=pelvis}}
     {targetBone {boneName=deform_glute_l}}
     {targetBone {boneName=thigh_twist_01_l}}
     {targetBone {boneName=spine_02}}
     {targetBone {boneName=spine_01}}
     {targetBone {boneName=spine_03}}
    ]
    constraintWeights [1,1,1,1,1,1]
   }
   n8 = Anim Dynamics - Bone: dyn_chain_01!(SimulationSpace:AnimPhysSimSpaceType Enum=Component, ComponentPose=@n7, Alpha=1, ExternalForce="0,0,0") -> (Pose) {
    angularDampingOverride=0.89
    angularSpringConstant=100
    bAngularSpring=true
    bGravityOverrideInSimSpace=true
    bOverrideAngularDamping=true
    bUseGravityOverride=true
    linearDampingOverride=0.9
    linearSpringConstant=150
    numSolverIterationsPreUpdate=3
    simulationSpace=BoneRelative
    alphaBoolBlend {blendOption=Linear}
    boundBone {boneName=dyn_chain_01}
    componentAppliedLinearAccClamp=<1000 1000 1000>
    componentLinearAccScale=<0.5 0.5 0.5>
    componentLinearVelScale=<0.5 0.5 0.5>
    gravityOverride {z=-980}
    physicsBodyDefinitions [{
      boundBone {boneName=dyn_chain_01}
      boxExtents=<17 17 17>
      constraintSetup {angularLimitsMax=<0 0 0>, angularLimitsMin=<0 -30 -10>, angularTarget=<1 0 0>}
      localJointOffset=<2 0 0>
     }]
    relativeSpaceBone {boneName=root}
   }
   n9 = Anim Dynamics - Bone: dyn_chain_02!(SimulationSpace:AnimPhysSimSpaceType Enum=BoneRelative, ComponentPose=@n8, Alpha=1, ExternalForce="0,0,0") -> (Pose) {
    angularDampingOverride=0.89
    angularSpringConstant=100
    bAngularSpring=true
    bGravityOverrideInSimSpace=true
    bOverrideAngularDamping=true
    bUseGravityOverride=true
    linearDampingOverride=0.9
    linearSpringConstant=150
    numSolverIterationsPreUpdate=3
    simulationSpace=BoneRelative
    alphaBoolBlend {blendOption=Linear}
    boundBone {boneName=dyn_chain_02}
    componentAppliedLinearAccClamp=<1000 1000 1000>
    componentLinearAccScale=<0.5 0.5 0.5>
    componentLinearVelScale=<0.5 0.5 0.5>
    gravityOverride {z=-980}
    physicsBodyDefinitions [{
      boundBone {boneName=dyn_chain_02}
      boxExtents=<17 17 17>
      constraintSetup {angularLimitsMax=<0 0 0>, angularLimitsMin=<0 -30 -10>, angularTarget=<1 0 0>}
      localJointOffset=<2 0 0>
     }]
    relativeSpaceBone {boneName=root}
   }
   n10 = Anim Dynamics - Bone: dyn_chain_03!(SimulationSpace:AnimPhysSimSpaceType Enum=BoneRelative, ComponentPose=@n9, Alpha=1, ExternalForce="0,0,0") -> (Pose) {
    angularDampingOverride=0.89
    angularSpringConstant=100
    bAngularSpring=true
    bGravityOverrideInSimSpace=true
    bOverrideAngularDamping=true
    bUseGravityOverride=true
    linearDampingOverride=0.9
    linearSpringConstant=150
    numSolverIterationsPreUpdate=3
    simulationSpace=BoneRelative
    alphaBoolBlend {blendOption=Linear}
    boundBone {boneName=dyn_chain_03}
    componentAppliedLinearAccClamp=<1000 1000 1000>
    componentLinearAccScale=<0.5 0.5 0.5>
    componentLinearVelScale=<0.5 0.5 0.5>
    gravityOverride {z=-980}
    physicsBodyDefinitions [{
      boundBone {boneName=dyn_chain_03}
      boxExtents=<17 17 17>
      constraintSetup {angularLimitsMax=<0 0 0>, angularLimitsMin=<0 -20 -10>, angularTarget=<1 0 0>}
      localJointOffset=<2 0 0>
     }]
    relativeSpaceBone {boneName=root}
   }
   n11 = Anim Dynamics - Bone: dyn_chain_05!(SimulationSpace:AnimPhysSimSpaceType Enum=BoneRelative, ComponentPose=@n10, Alpha=1, ExternalForce="0,0,0") -> (Pose) {
    angularDampingOverride=0.89
    angularSpringConstant=100
    bAngularSpring=true
    bGravityOverrideInSimSpace=true
    bOverrideAngularDamping=true
    bUseGravityOverride=true
    linearDampingOverride=0.9
    linearSpringConstant=150
    numSolverIterationsPreUpdate=3
    simulationSpace=BoneRelative
    alphaBoolBlend {blendOption=Linear}
    boundBone {boneName=dyn_chain_05}
    componentAppliedLinearAccClamp=<1000 1000 1000>
    componentLinearAccScale=<0.5 0.5 0.5>
    componentLinearVelScale=<0.5 0.5 0.5>
    gravityOverride {z=-980}
    physicsBodyDefinitions [{
      boundBone {boneName=dyn_chain_05}
      boxExtents=<17 17 17>
      constraintSetup {angularLimitsMax=<0 0 0>, angularLimitsMin=<0 -10 -10>, angularTarget=<1 0 0>}
      localJointOffset=<-2 0 0>
     }]
    relativeSpaceBone {boneName=root}
   }
   n12 = Anim Dynamics - Bone: dyn_chain_06!(SimulationSpace:AnimPhysSimSpaceType Enum=BoneRelative, ComponentPose=@n11, Alpha=1, ExternalForce="0,0,0") -> (Pose) {
    angularDampingOverride=0.89
    angularSpringConstant=100
    bAngularSpring=true
    bGravityOverrideInSimSpace=true
    bOverrideAngularDamping=true
    bUseGravityOverride=true
    linearDampingOverride=0.9
    linearSpringConstant=150
    numSolverIterationsPreUpdate=3
    simulationSpace=BoneRelative
    alphaBoolBlend {blendOption=Linear}
    boundBone {boneName=dyn_chain_06}
    componentAppliedLinearAccClamp=<1000 1000 1000>
    componentLinearAccScale=<0.5 0.5 0.5>
    componentLinearVelScale=<0.5 0.5 0.5>
    gravityOverride {z=-980}
    physicsBodyDefinitions [{
      boundBone {boneName=dyn_chain_06}
      boxExtents=<17 17 17>
      constraintSetup {angularLimitsMax=<0 0 0>, angularLimitsMin=<0 -10 -10>, angularTarget=<1 0 0>}
      localJointOffset=<-2 0 0>
     }]
    relativeSpaceBone {boneName=root}
   }
   n13 = Anim Dynamics - Bone: dyn_chain_07!(SimulationSpace:AnimPhysSimSpaceType Enum=BoneRelative, ComponentPose=@n12, Alpha=1, ExternalForce="0,0,0") -> (Pose) {
    angularDampingOverride=0.89
    angularSpringConstant=100
    bAngularSpring=true
    bGravityOverrideInSimSpace=true
    bOverrideAngularDamping=true
    bUseGravityOverride=true
    linearDampingOverride=0.9
    linearSpringConstant=150
    numSolverIterationsPreUpdate=3
    simulationSpace=BoneRelative
    alphaBoolBlend {blendOption=Linear}
    boundBone {boneName=dyn_chain_07}
    componentAppliedLinearAccClamp=<1000 1000 1000>
    componentLinearAccScale=<0.5 0.5 0.5>
    componentLinearVelScale=<0.5 0.5 0.5>
    gravityOverride {z=-980}
    physicsBodyDefinitions [{
      boundBone {boneName=dyn_chain_07}
      boxExtents=<17 17 17>
      constraintSetup {angularLimitsMax=<0 0 0>, angularLimitsMin=<0 -20 -10>, angularTarget=<1 0 0>}
      localJointOffset=<-2 0 0>
     }]
    relativeSpaceBone {boneName=root}
   }
   n14 = "Constraint - dyn_chain_c (PPP)"!(ComponentPose=@n13, Alpha=1, ConstraintWeights_0=0.5, ConstraintWeights_1:float=0.5, ConstraintWeights_2:float=0) -> (Pose) {
    alphaBoolBlend {blendOption=Linear}
    boneToModify {boneName=dyn_chain_c}
    constraintSetup [{targetBone {boneName=dyn_chain_03}}, {targetBone {boneName=dyn_chain_07}}, {targetBone {boneName=thigh_l}}]
    constraintWeights [1,1,1]
   }
   n15 = Anim Dynamics - Bone: dyn_chain_c!(SimulationSpace:AnimPhysSimSpaceType Enum=BoneRelative, ComponentPose=@n14, Alpha=1, ExternalForce="0,0,0") -> (Pose) {
    angularDampingOverride=0.89
    angularSpringConstant=100
    bAngularSpring=true
    bGravityOverrideInSimSpace=true
    bOverrideAngularDamping=true
    bUseGravityOverride=true
    linearDampingOverride=0.9
    linearSpringConstant=150
    numSolverIterationsPreUpdate=3
    simulationSpace=BoneRelative
    alphaBoolBlend {blendOption=Linear}
    boundBone {boneName=dyn_chain_c}
    componentAppliedLinearAccClamp=<10000 10000 10000>
    componentLinearAccScale=<0.5 0.5 0.5>
    componentLinearVelScale=<0.5 0.5 0.5>
    gravityOverride {z=-980}
    physicsBodyDefinitions [{
      boundBone {boneName=dyn_chain_c}
      boxExtents=<17 17 17>
      constraintSetup {angularLimitsMax=<0 0 5>, angularLimitsMin=<0 -30 -5>, angularTarget=<1 0 0>}
      localJointOffset=<2 0 0>
     }]
    relativeSpaceBone {boneName=root}
   }
   n16 = "Constraint - dyn_grenade_upper_mstr (PPPPPPPPPP)"!(
    ComponentPose=@n15
    Alpha=1
    ConstraintWeights_0=0.3145
    ConstraintWeights_1:float=0.1946
    ConstraintWeights_2:float=0.1268
    ConstraintWeights_3:float=0.1068
    ConstraintWeights_4:float=0.0964
    ConstraintWeights_5:float=0.0792
    ConstraintWeights_6:float=0.052
    ConstraintWeights_7:float=0.0139
    ConstraintWeights_8:float=0.0113
    ConstraintWeights_9:float=0.0045
   ) -> (Pose) {
    alphaBoolBlend {blendOption=Linear}
    boneToModify {boneName=dyn_grenade_upper_mstr}
    constraintSetup [
     {targetBone {boneName=deform_pec_r}}
     {targetBone {boneName=clavicle_r}}
     {targetBone {boneName=spine_05}}
     {targetBone {boneName=deform_upperarm_fr_r}}
     {targetBone {boneName=spine_04}}
     {targetBone {boneName=upperarm_twist_01_r}}
     {targetBone {boneName=spine_03}}
     {targetBone {boneName=deform_upperarm_up_r}}
     {targetBone {boneName=spine_02}}
     {targetBone {boneName=upperarm_twist_02_r}}
    ]
    constraintWeights [1,1,1,1,1,1,1,1,1,1]
   }
   n17 = Anim Dynamics - Bone: dyn_grenade_upper(ComponentPose=@n16, Alpha=1, ExternalForce="0,0,0") {
    angularDampingOverride=1
    angularSpringConstant=20000
    bAngularSpring=true
    bOverrideAngularDamping=true
    bUseGravityOverride=true
    gravityScale=0.35
    simulationSpace=RootRelative
    alphaBoolBlend {blendOption=Linear}
    boundBone {boneName=dyn_grenade_upper}
    componentLinearAccScale=<1 1 1>
    componentLinearVelScale=<1 1 1>
    gravityOverride {z=-980}
    physicsBodyDefinitions [{
      boundBone {boneName=dyn_grenade_upper}
      boxExtents=<20 30 30>
      constraintSetup {angularLimitsMax=<15 15 15>, angularLimitsMin=<-15 -15 -15>, angularTarget=<1 0 0>}
      localJointOffset=<4 0 0>
     }]
   }
   n18 = "Constraint - dyn_grenade_lower (PPPPPPPPP)"!(
    ComponentPose=@n17
    Alpha=1
    ConstraintWeights_0=0.4132
    ConstraintWeights_1:float=0.1627
    ConstraintWeights_2:float=0.1624
    ConstraintWeights_3:float=0.0924
    ConstraintWeights_4:float=0.0778
    ConstraintWeights_5:float=0.052
    ConstraintWeights_6:float=0.0209
    ConstraintWeights_7:float=0.0148
    ConstraintWeights_8:float=0.0038
   ) -> (Pose) {
    alphaBoolBlend {blendOption=Linear}
    boneToModify {boneName=dyn_grenade_lower}
    constraintSetup [
     {targetBone {boneName=deform_pec_r}}
     {targetBone {boneName=spine_03}}
     {targetBone {boneName=spine_04}}
     {targetBone {boneName=clavicle_r}}
     {targetBone {boneName=spine_02}}
     {targetBone {boneName=spine_05}}
     {targetBone {boneName=deform_upperarm_fr_r}}
     {targetBone {boneName=upperarm_twist_01_r}}
     {targetBone {boneName=spine_01}}
    ]
    constraintWeights [1,1,1,1,1,1,1,1,1]
   }
   n19 = Anim Dynamics - Bone: dyn_grenade_lower(ComponentPose=@n18, Alpha=1, ExternalForce="0,0,0") {
    angularDampingOverride=1
    angularSpringConstant=20000
    bAngularSpring=true
    bOverrideAngularDamping=true
    bUseGravityOverride=true
    gravityScale=0.35
    simulationSpace=RootRelative
    alphaBoolBlend {blendOption=Linear}
    boundBone {boneName=dyn_grenade_lower}
    componentLinearAccScale=<1 1 1>
    componentLinearVelScale=<1 1 1>
    gravityOverride {z=-980}
    physicsBodyDefinitions [{
      boundBone {boneName=dyn_grenade_lower}
      boxExtents=<20 30 30>
      constraintSetup {angularLimitsMax=<15 15 15>, angularLimitsMin=<-15 -15 -15>, angularTarget=<1 0 0>}
      localJointOffset=<4 0 0>
     }]
   }
   n20 = Anim Dynamics - Bone: dyn_jacket_l_01(ComponentPose=@n19, Alpha=0.7, ExternalForce="0,0,0") {
    angularDampingOverride=0.999
    bOverrideAngularDamping=true
    bUseGravityOverride=true
    simulationSpace=RootRelative
    alphaBoolBlend {blendOption=Linear}
    boundBone {boneName=dyn_jacket_l_01}
    componentAppliedLinearAccClamp=<3000 3000 3000>
    componentLinearAccScale=<0.5 0.5 0.5>
    componentLinearVelScale=<0.5 0.5 0.5>
    gravityOverride {z=-980}
    physicsBodyDefinitions [{
      boundBone {boneName=dyn_jacket_l_01}
      boxExtents=<10 10 10>
      constraintSetup {angularLimitsMax=<0 15 30>, angularLimitsMin=<0 -15 0>, angularTarget=<0 0 0>}
      localJointOffset=<1 0 0>
     }]
   }
   n21 = Anim Dynamics - Bone: dyn_jacket_l_02(ComponentPose=@n20, Alpha=0.7, ExternalForce="0,0,0") {
    angularDampingOverride=0.999
    bOverrideAngularDamping=true
    bUseGravityOverride=true
    simulationSpace=RootRelative
    alphaBoolBlend {blendOption=Linear}
    boundBone {boneName=dyn_jacket_l_02}
    componentAppliedLinearAccClamp=<3000 3000 3000>
    componentLinearAccScale=<0.5 0.5 0.5>
    componentLinearVelScale=<0.5 0.5 0.5>
    gravityOverride {z=-980}
    physicsBodyDefinitions [{
      boundBone {boneName=dyn_jacket_l_02}
      boxExtents=<10 10 10>
      constraintSetup {angularLimitsMax=<0 15 30>, angularLimitsMin=<0 -15 0>, angularTarget=<0 0 0>}
      localJointOffset=<1 0 0>
     }]
   }
   n22 = Anim Dynamics - Bone: dyn_jacket_r_01(ComponentPose=@n21, Alpha=0.7, ExternalForce="0,0,0") {
    angularDampingOverride=0.999
    bOverrideAngularDamping=true
    bUseGravityOverride=true
    simulationSpace=RootRelative
    alphaBoolBlend {blendOption=Linear}
    boundBone {boneName=dyn_jacket_r_01}
    componentAppliedLinearAccClamp=<3000 3000 3000>
    componentLinearAccScale=<0.5 0.5 0.5>
    componentLinearVelScale=<0.5 0.5 0.5>
    gravityOverride {z=-980}
    physicsBodyDefinitions [{
      boundBone {boneName=dyn_jacket_r_01}
      boxExtents=<10 10 10>
      constraintSetup {angularLimitsMax=<0 15 30>, angularLimitsMin=<0 -15 0>, angularTarget=<0 0 0>}
      localJointOffset=<1 0 0>
     }]
   }
   n23 = Anim Dynamics - Bone: dyn_jacket_r_02(ComponentPose=@n22, Alpha=0.7, ExternalForce="0,0,0") {
    angularDampingOverride=0.999
    bOverrideAngularDamping=true
    bUseGravityOverride=true
    simulationSpace=RootRelative
    alphaBoolBlend {blendOption=Linear}
    boundBone {boneName=dyn_jacket_r_02}
    componentAppliedLinearAccClamp=<3000 3000 3000>
    componentLinearAccScale=<0.5 0.5 0.5>
    componentLinearVelScale=<0.5 0.5 0.5>
    gravityOverride {z=-980}
    physicsBodyDefinitions [{
      boundBone {boneName=dyn_jacket_r_02}
      boxExtents=<10 10 10>
      constraintSetup {angularLimitsMax=<0 15 30>, angularLimitsMin=<0 -15 0>, angularTarget=<0 0 0>}
      localJointOffset=<1 0 0>
     }]
   }
   n24 = Anim Dynamics - Bone: dyn_jacket_back_01!(ComponentPose=@n23, bAlphaBoolEnabled:bool=True, ExternalForce="0,0,0") -> (Pose) {
    alphaInputType=Bool
    angularDampingOverride=0.999
    angularSpringConstant=300
    bAngularSpring=true
    bOverrideAngularDamping=true
    bUseGravityOverride=true
    simulationSpace=RootRelative
    alphaBoolBlend {blendInTime=0.1, blendOption=Linear, blendOutTime=0.1}
    boundBone {boneName=dyn_jacket_back_01}
    componentAppliedLinearAccClamp=<3000 3000 3000>
    componentLinearAccScale=<0.4 0.4 0.4>
    componentLinearVelScale=<0.4 0.4 0.4>
    gravityOverride {z=-980}
    physicsBodyDefinitions [{
      boundBone {boneName=dyn_jacket_back_01}
      boxExtents=<10 10 10>
      constraintSetup {angularLimitsMax=<5 5 10>, angularLimitsMin=<-5 -5 0>, angularTarget=<0 0 0>}
      localJointOffset=<1 0 0>
     }]
   }
   n25 = Anim Dynamics - Bone: dyn_lapel_r(ComponentPose=@n24, Alpha=Get NotTargetingAlpha() -> (NotTargetingAlpha:double), ExternalForce="0,0,0") {
    angularDampingOverride=0.9999
    angularSpringConstant=500
    bAngularSpring=true
    bOverrideAngularDamping=true
    bUseGravityOverride=true
    simulationSpace=RootRelative
    alphaBoolBlend {blendOption=Linear}
    boundBone {boneName=dyn_lapel_r}
    componentAppliedLinearAccClamp=<5000 5000 5000>
    componentLinearAccScale=<0.4 0.4 0.4>
    componentLinearVelScale=<0.4 0.4 0.4>
    gravityOverride {z=-980}
    physicsBodyDefinitions [{
      boundBone {boneName=dyn_lapel_r}
      boxExtents=<6 6 6>
      constraintSetup {angularLimitsMax=<0 0 5>, angularLimitsMin=<0 -7 -10>, angularTarget=<1 0 0>}
      localJointOffset=<0 0 0>
     }]
   }
   n26 = Anim Dynamics - Bone: dyn_lapel_l(ComponentPose=@n25, Alpha=1, ExternalForce="0,0,0") {
    angularDampingOverride=0.9999
    angularSpringConstant=500
    bAngularSpring=true
    bOverrideAngularDamping=true
    bUseGravityOverride=true
    simulationSpace=RootRelative
    alphaBoolBlend {blendOption=Linear}
    boundBone {boneName=dyn_lapel_l}
    componentAppliedLinearAccClamp=<5000 5000 5000>
    componentLinearAccScale=<0.4 0.4 0.4>
    componentLinearVelScale=<0.4 0.4 0.4>
    gravityOverride {z=-980}
    physicsBodyDefinitions [{
      boundBone {boneName=dyn_lapel_l}
      boxExtents=<6 6 6>
      constraintSetup {angularLimitsMax=<0 0 5>, angularLimitsMin=<0 -7 -10>, angularTarget=<1 0 0>}
      localJointOffset=<0 0 0>
     }]
   }
   n27 = Anim Dynamics - Bone: dyn_necklace(ComponentPose=@n26, Alpha=0.5, ExternalForce="0,0,0") {
    angularDampingOverride=1
    angularSpringConstant=2000
    bOverrideAngularDamping=true
    bUseGravityOverride=true
    simulationSpace=BoneRelative
    alphaBoolBlend {blendOption=Linear}
    boundBone {boneName=dyn_necklace}
    componentAppliedLinearAccClamp=<10000 10000 10000>
    componentLinearAccScale=<0.5 0.5 0.5>
    componentLinearVelScale=<0.5 0.5 0.5>
    gravityOverride {z=-980}
    physicsBodyDefinitions [{
      boundBone {boneName=dyn_necklace}
      boxExtents=<13 13 13>
      constraintSetup {angularLimitsMax=<0 0 2>, angularLimitsMin=<0 -20 -2>, angularTarget=<1 0 0>}
      localJointOffset=<6 0 0>
     }]
    relativeSpaceBone {boneName=spine_05}
   }
   n28 = Anim Dynamics - Bone: dyn_necklace_charm(ComponentPose=@n27, Alpha=1, ExternalForce="0,0,0") {
    angularDampingOverride=0.9999
    angularSpringConstant=50
    bOverrideAngularDamping=true
    bUseGravityOverride=true
    simulationSpace=BoneRelative
    alphaBoolBlend {blendOption=Linear}
    boundBone {boneName=dyn_necklace_charm}
    componentAppliedLinearAccClamp=<10000 10000 10000>
    componentLinearAccScale=<0.5 0.5 0.5>
    componentLinearVelScale=<0.5 0.5 0.5>
    gravityOverride {z=-980}
    physicsBodyDefinitions [{
      boundBone {boneName=dyn_necklace_charm}
      boxExtents=<13 13 13>
      constraintSetup {angularLimitsMax=<0 0 5>, angularLimitsMin=<0 -50 -5>, angularTarget=<1 0 0>}
      localJointOffset=<6 0 0>
     }]
    relativeSpaceBone {boneName=root}
   }
   n29 = "Transform (Modify) Bone - Bone: dyn_jacket_back_01"(ComponentPose=@n28, Alpha=Get Jumping() -> (Jumping:double), Translation="0,0,0", Rotation="0,0,25", Scale="1,1,1") {
    rotationMode=BMM_Additive
    translationMode=BMM_Additive
    alphaBoolBlend {blendOption=Linear}
    alphaScaleBiasClamp {bInterpResult=true, interpSpeedDecreasing=7, interpSpeedIncreasing=5}
    boneToModify {boneName=dyn_jacket_back_01}
    rotation {roll=25}
   }
   n30 = "Transform (Modify) Bone - Bone: dyn_shoulder_MSTR"(ComponentPose=@n29, Alpha=Get Melee() -> (Melee:double), Translation="0,0,0", Rotation="0,-10,0", Scale="1,1,1") {
    rotationMode=BMM_Additive
    translationMode=BMM_Additive
    alphaBoolBlend {blendOption=Linear}
    alphaScaleBiasClamp {bInterpResult=true, interpSpeedDecreasing=7, interpSpeedIncreasing=5}
    boneToModify {boneName=dyn_shoulder_MSTR}
    rotation {yaw=-10}
   }
   n31 = Get FE Custom Lobby Idle Is Playing Float() -> (FE Custom Lobby Idle Is Playing Float:float)
   n32 = "Transform (Modify) Bone - Bone: dyn_jacket_l_01"(ComponentPose=@n30, Alpha=@n31, Translation="0,0,0", Rotation="0,0,-5", Scale="1,1,1") {
    rotationMode=BMM_Additive
    translationMode=BMM_Additive
    alphaBoolBlend {blendOption=Linear}
    alphaScaleBiasClamp {bInterpResult=true, interpSpeedDecreasing=7, interpSpeedIncreasing=5}
    boneToModify {boneName=dyn_jacket_l_01}
    rotation {roll=-5}
   }
   n33 = "Transform (Modify) Bone - Bone: dyn_lapel_r_mstr"(ComponentPose=@n32, Alpha=@n31, Translation="0,0,0", Rotation="5,-10,5", Scale="1,1,1") {
    rotationMode=BMM_Additive
    translationMode=BMM_Additive
    alphaBoolBlend {blendOption=Linear}
    alphaScaleBiasClamp {bInterpResult=true, interpSpeedDecreasing=7, interpSpeedIncreasing=5}
    boneToModify {boneName=dyn_lapel_r_mstr}
    rotation=<rot 5 -10 5>
   }
   n34 = "Transform (Modify) Bone - Bone: dyn_neck_back"(ComponentPose=@n33, Alpha=@n31, Translation="1,0,-1", Rotation="0,0,0", Scale="1,1,1") {
    rotationMode=BMM_Additive
    translationMode=BMM_Additive
    alphaBoolBlend {blendOption=Linear}
    alphaScaleBiasClamp {bInterpResult=true, interpSpeedDecreasing=7, interpSpeedIncreasing=5}
    boneToModify {boneName=dyn_neck_back}
    translation {x=1, z=-1}
   }
   n35 = Control Rig!(
    Source=Component To Local(ComponentPose=@n34)
    Alpha
    Fly_mode:double=Get Dyn_fly() -> (Dyn_fly:float)
    Pistol:double=Get pistol() -> (pistol:double)
    Rifle:double=Get Rifle() -> (Rifle:double)
    Crouch:double=Get Crouch() -> (Crouch:double)
    SkyDive:double=Get Skydive() -> (Skydive:double)
    Targeting:double=Get ADS() -> (ADS:double)
    Parachute:double=Get parachute() -> (parachute:double)
    Emote:bool=Get bIsPlayingEmote()
   ) -> (Pose) {
    alphaBoolBlend {blendOption=Linear}
    controlRigAssetReference {blueprintRigClass=&/M_LRG_BakerStep_CtrlRig.M_LRG_BakerStep_CtrlRig_C}
   }
   Output Pose(Result=Layered blend per bone(BasePose=@n35, BlendPoses_0=Slot 'DefaultSlot'(), BlendWeights_0=Get FE Custom Lobby Idle Is Playing Float()) {layerSetup [{branchFilters [{boneName=dyn_shoulder_MSTR}]}]}) {layerGroup=None}
  ]}
 EventGraph {nodes [
   Event BlueprintUpdateAnimation()
   > n1 = Sequence!(execute) -> (then_0, then_1, then_2:Exec, then_3:Exec, then_4:Exec)
   > n2 = Branch(Condition=Get bIsInFrontEnd())
   > Set FrontendAlpha(FrontendAlpha:double=0.3) -> (Output_Get:double)
   > Set Simulation Space(Simulation Space:AnimPhysSimSpaceType Enum=World) -> (Output_Get:AnimPhysSimSpaceType Enum)
   > Set LayeredFE_DynAlpha(LayeredFE_DynAlpha:double=0.5) -> (Output_Get:double)
   Set FrontendAlpha(execute=@n2.else, FrontendAlpha=1)
   > Set Simulation Space(Simulation Space=BoneRelative)
   > Set LayeredFE_DynAlpha(LayeredFE_DynAlpha=0)
   n3 = Sequence!(execute=@n1.then_2) -> (then_0, then_1)
   > n4 = Branch(Condition=Get bIsSkyDiving())
   > n5 = Branch(Condition=Get bIsSkydiveFloating())
   > Set Dyn_fly(Dyn_fly:float=0.75) -> (Output_Get:float)
   Set Dyn_fly(execute=@n5.else, Dyn_fly=1)
   Set Dyn_fly(execute=@n4.else, Dyn_fly=0)
   n6 = Branch(execute=@n3.then_1, Condition=Get bIsInFrontEnd())
   > Set Dyn_Frontend_alfa(Dyn_Frontend_alfa:float=1) -> (Output_Get:float)
   Set Dyn_Frontend_alfa(execute=@n6.else, Dyn_Frontend_alfa=0)
   Set Skydive(execute=@n1.then_3, Skydive:double=SelectFloat(A=1, B=0, bPickA=Get bIsSkyDiving())) -> (Output_Get:double)
   > Set ADS(ADS:double=SelectFloat(A=1, B=0, bPickA=Get bIsTargeting())) -> (Output_Get:double)
   > Set Emote(Emote:double=SelectFloat(A=1, B=0, bPickA=Get bIsPlayingEmote())) -> (Output_Get:double)
   > Set moving(moving:double=SelectFloat(A=1, B=0, bPickA=Get bIsMoving2D())) -> (Output_Get:double)
   > Set Crouch(Crouch:double=SelectFloat(A=1, B=0, bPickA=Get bIsCrouching())) -> (Output_Get:double)
   > Set parachute(parachute:double=SelectFloat(A=1, B=0, bPickA=Get bIsParachuteOpened())) -> (Output_Get:double)
   > Set Emote(Emote=SelectFloat(A=1, B=0, bPickA=Get bIsPlayingEmote()))
   n7 = Branch(execute=@n1.then_4, Condition=Get bIsInFrontEnd())
   > Set FE Custom Lobby Idle Is Playing Float(FE Custom Lobby Idle Is Playing Float:float=GetCurveValue(self=GetOwnerFrontEndAnimInstance(), CurveName=Frontend_Idle_Bypass)) -> (Output_Get:float)
   > Set FrontEnd Idle Alpha(FrontEnd Idle Alpha:double=0.5) -> (Output_Get:double)
   Set FE Custom Lobby Idle Is Playing Float(execute=@n7.else, FE Custom Lobby Idle Is Playing Float=0)
   > Set FrontEnd Idle Alpha(FrontEnd Idle Alpha=0)
   n8 = Gravity Override State Function DEPRECATED(
    execute=@n1.then_1
    CharacterPartAnimInstance=Self-Reference()
    GameTimeInSeconds=GetGameTimeInSeconds()
    GravityOverrideParams=Get Gravity Override Params()
   )
   > Set GravityOverride(GravityOverride=@n8)
   Event OnBeginJumping()
   > Set Jumping(Jumping:double=1) -> (Output_Get:double)
   Event OnWeaponChanged()
   > n9 = Cast To FortPlayerAnimInstance(Object=GetAnimInstance(self=GetOwnerSkeletalMeshComponent()))
   > Set pistol(pistol:double=0) -> (Output_Get:double)
   > Set Melee(Melee:double=0) -> (Output_Get:double)
   > Set Rifle(Rifle:double=0) -> (Output_Get:double)
   > n10 = Switch on EFortWeaponCoreAnimation!(execute, Selection:EFortWeaponCoreAnimation Enum=Get WeaponCoreAnim(self=@n9)) -> (
    Melee:Exec
    Pistol:Exec
    Shotgun:Exec
    PaperBlueprint:Exec
    Rifle:Exec
    MeleeOneHand:Exec
    MachinePistol:Exec
    RocketLauncher:Exec
    GrenadeLauncher:Exec
    GoingCommando:Exec
    AssaultRifle:Exec
    TacticalShotgun:Exec
    SniperRifle:Exec
    TrapPlacement:Exec
    ShoulderLauncher:Exec
    AbilityDecoTool:Exec
    Crossbow:Exec
    C4:Exec
    RemoteControl:Exec
    DualWield:Exec
    AR_BullPup:Exec
    AR_ForwardGrip:Exec
    MedPackPaddles:Exec
    SMG_P90:Exec
    AR_DrumGun:Exec
    Consumable_Small:Exec
    Consumable_Large:Exec
    Balloon:Exec
    MountedTurret:Exec
    CreativeTool:Exec
    ExplosiveBow:Exec
    AshtonIndigo:Exec
    AshtonChicago:Exec
    MeleeDualWield:Exec
    Unarmed:Exec
   )
   Set Rifle(
    execute=[
     @n10.AR_BullPup
     @n10.AR_ForwardGrip
     @n10.AssaultRifle
     @n10.Crossbow
     @n10.GrenadeLauncher
     @n10.Rifle
     @n10.RocketLauncher
     @n10.Shotgun
     @n10.SniperRifle
     @n10.TacticalShotgun
    ]
    Rifle=1
   )
   Set Melee(execute=@n10.Melee, Melee=1)
   Set pistol(execute=@n10.Pistol, pistol=1)
   Event OnBeginFalling()
   > Set Jumping(Jumping=1)
   Event OnEndFalling()
   > Set Jumping(Jumping=0)
   n11 = Event OnCharacterCustomizationLoaded()
   > Set HeadMesh(HeadMesh:Skeletal Mesh Component Budgeted Object Reference=Get Array Item with EFort Custom Part Type(Array=@n11.CustomMeshParts, PartType=Head)) -> (Output_Get:Skeletal Mesh Component Budgeted Object Reference)
   n12 = CheckForBackbling()
   > Set bIsNotWearingBackpack(bIsNotWearingBackpack:bool=NOT Boolean(A=OR Boolean(A=@n12, B=@n12.bIsBackblingLong))) -> (Output_Get:bool)
   Event OnBeginTargeting() {is=title K2Node_Event}
   > n13 = Sequence()
   > Set Targeting(Targeting:double=1) -> (Output_Get:double)
   Set NotTargetingAlpha(execute=@n13.then_1, NotTargetingAlpha:double=0) -> (Output_Get:double)
   Event OnEndTargeting() {is=title K2Node_Event}
   > n14 = Sequence()
   > Set NotTargetingAlpha(NotTargetingAlpha=1)
   Set Targeting(execute=@n14.then_1, Targeting=0)
   Event OnEndJumping()
   > Set Jumping(Jumping=0)
  ]}
]
variables [
 ADS {type=number}
 bIsNotWearingBackpack {type=boolean}
 Crouch {type=number}
 Dyn_fly {type=number}
 Dyn_Frontend_alfa {type=number}
 Emote {type=number}
 FE Custom Lobby Idle Is Playing Float {type=number, default=1}
 FrontEnd Idle Alpha {type=number}
 FrontendAlpha {type=number}
 Gravity Override Params {type=GravityOverrideParamsStruct}
 GravityOverride {type=Vector}
 HeadMesh {type=SkeletalMeshComponentBudgeted}
 Jump {type=number}
 Jumping {type=number}
 LayeredFE_DynAlpha {type=number}
 Melee {type=number}
 moving {type=number}
 NotTargetingAlpha {type=number}
 parachute {type=number}
 pistol {type=number}
 Rifle {type=number}
 Simulation Space {type=AnimPhysSimSpaceType}
 Skydive {type=number}
 Targeting {type=number}
]
