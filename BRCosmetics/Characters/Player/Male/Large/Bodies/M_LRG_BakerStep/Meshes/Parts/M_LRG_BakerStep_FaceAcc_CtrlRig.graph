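# Control Rig for the BakerStep large male face accessory that adjusts braid and hair bones as the head moves.
# A sequence runs four spherical pose readers on the head relative to spine_05, each weighting a Modify Transforms on bones including the braid master.
# The fourth reader draws debug and its transforms are all identity, so it applies no visible offset.
# The braid controls and declared variables are unused by the graph.
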ControlRigBlueprint /BRCosmetics/Characters/Player/Male/Large/Bodies/M_LRG_BakerStep/Meshes/Parts/M_LRG_BakerStep_FaceAcc_CtrlRig
graphs [
 RigVMFunctionLibrary {links [], nodes []}
 RigVMFunction_Sequence_1_ContainedGraph {nodes [
   n1 = RigVMFunctionEntryNode()
   RigVMFunction_Sequence = &RigVMFunction_Sequence(ExecuteContext=@n1.ExecuteContext)
   n2 = &RigVMFunction_Sequence(ExecuteContext=@RigVMFunction_Sequence.B)
   n3 = &RigVMFunction_Sequence(ExecuteContext=@n2.B)
   RigVMFunctionReturnNode(A=@RigVMFunction_Sequence.A, B=@n2.A, C=@n3.A, D=@n3.B)
  ]}
 RigVMModel {nodes [
   RigUnit_BeginExecution = &RigUnit_BeginExecution()
   RigVMFunction_Sequence = RigVMAggregateNode(ExecuteContext=@RigUnit_BeginExecution.ExecutePin)
   n1 = &RigUnit_SphericalPoseReader(
    ActiveRegionScaleFactors=(PositiveWidth=1,NegativeWidth=1,PositiveHeight=1,NegativeHeight=1)
    ActiveRegionSize=0.1
    Debug=(bDrawDebug=false,bDraw2D=False,bDrawLocalAxes=False,DebugScale=25,DebugSegments=20,DebugThickness=0.25)
    DriverAxis=(X=1,Y=0,Z=0)
    DriverItem=(Type=Bone,Name="head")
    FalloffRegionScaleFactors=(PositiveWidth=1,NegativeWidth=1,PositiveHeight=1,NegativeHeight=1)
    FalloffSize=0.25
    FlipHeightScaling=False
    FlipWidthScaling=False
    OptionalParentItem=(Type=Bone,Name="spine_05")
    RotationOffset=(X=0,Y=-45,Z=0)
    ExecutePin=@RigVMFunction_Sequence.A
   )
   > &RigUnit_ModifyTransforms(
    ItemToModify="((Item=(Type=Bone,Name=\"dyn_braid_MSTR\"),Transform=(Rotation=(X=0,Y=0.0814,Z=0,W=0.9967),Translation=(X=0,Y=0,Z=0),Scale3D=(X=0.9688,Y=0.9688,Z=0.9688))),(Item=(Type=Bone,Name=\"dyn_head_plate_lower\"),Transform=(Rotation=(X=0,Y=0.0593,Z=0,W=0.9982),Translat ...(truncated)"
    WeightMinimum=0
    ItemToModify.0.Transform.Rotation=&RigVMFunction_MathQuaternionFromEuler(Euler=(X=0,Y=0,Z=0)).Result
    Weight=@n1.OutputParam
   )
   n2 = &RigUnit_SphericalPoseReader(
    ActiveRegionScaleFactors=(PositiveWidth=2,NegativeWidth=0.5,PositiveHeight=2,NegativeHeight=2)
    ActiveRegionSize=0.075
    Debug=(bDrawDebug=false,DebugScale=25,DebugSegments=20,DebugThickness=0.25)
    DriverAxis=(X=1,Y=0,Z=0)
    DriverItem=(Type=Bone,Name="head")
    FalloffRegionScaleFactors=(PositiveWidth=1,NegativeWidth=0.4,PositiveHeight=1,NegativeHeight=1)
    FalloffSize=0.25
    FlipHeightScaling=false
    FlipWidthScaling=true
    OptionalParentItem=(Type=Bone,Name="spine_05")
    RotationOffset=(X=-70,Y=5,Z=0)
    ExecutePin=@RigVMFunction_Sequence.B
   )
   > &RigUnit_ModifyTransforms(
    ItemToModify="((Item=(Type=Bone,Name=\"dyn_braid_MSTR\"),Transform=(Rotation=(X=0,Y=0,Z=0.0698,W=0.9976),Translation=(X=-3,Y=0,Z=-5),Scale3D=(X=1,Y=1,Z=1))),(Item=(Type=Bone,Name=\"dyn_hair_neck\"),Transform=(Rotation=(X=0,Y=0,Z=0,W=1),Translation=(X=0,Y=0,Z=0),Scale3D=(X=1 ...(truncated)"
    WeightMinimum=0
    Weight=@n2.OutputParam
   )
   n3 = &RigUnit_SphericalPoseReader(
    ActiveRegionScaleFactors=(PositiveWidth=1,NegativeWidth=1,PositiveHeight=2,NegativeHeight=2)
    ActiveRegionSize=0.1
    Debug=(bDrawDebug=false,bDraw2D=False,bDrawLocalAxes=False,DebugScale=25,DebugSegments=20,DebugThickness=0.25)
    DriverAxis=(X=1,Y=0,Z=0)
    DriverItem=(Type=Bone,Name="head")
    FalloffRegionScaleFactors=(PositiveWidth=1,NegativeWidth=1,PositiveHeight=1,NegativeHeight=1)
    FalloffSize=0.35
    FlipHeightScaling=False
    FlipWidthScaling=False
    OptionalParentItem=(Type=Bone,Name="spine_05")
    RotationOffset=(X=-90,Y=0,Z=20)
    ExecutePin=@RigVMFunction_Sequence.C
   )
   > &RigUnit_ModifyTransforms(
    ItemToModify="((Item=(Type=Bone,Name=\"dyn_braid_MSTR\"),Transform=(Rotation=(X=0,Y=0,Z=0.0698,W=0.9976),Translation=(X=-3,Y=0,Z=-1),Scale3D=(X=1,Y=1,Z=1))),(Item=(Type=Bone,Name=\"dyn_hair_neck\"),Transform=(Rotation=(X=0,Y=0,Z=0,W=1),Translation=(X=-7,Y=0,Z=0),Scale3D=(X= ...(truncated)"
    WeightMinimum=0
    ItemToModify.0.Transform.Rotation=&RigVMFunction_MathQuaternionFromEuler(Euler=(X=0,Y=0,Z=0)).Result
    Weight=@n3.OutputParam
   )
   n4 = &RigUnit_SphericalPoseReader(
    ActiveRegionScaleFactors=(PositiveWidth=1,NegativeWidth=1,PositiveHeight=1,NegativeHeight=1)
    ActiveRegionSize=0.1
    Debug=(bDrawDebug=true,bDraw2D=False,bDrawLocalAxes=False,DebugScale=25,DebugSegments=20,DebugThickness=0.25)
    DriverAxis=(X=1,Y=0,Z=0)
    DriverItem=(Type=Bone,Name="head")
    FalloffRegionScaleFactors=(PositiveWidth=1,NegativeWidth=1,PositiveHeight=1,NegativeHeight=1)
    FalloffSize=0.25
    FlipHeightScaling=False
    FlipWidthScaling=False
    OptionalParentItem=(Type=Bone,Name="spine_05")
    RotationOffset=(X=-60,Y=0,Z=20)
    ExecutePin=@RigVMFunction_Sequence.D
   )
   > &RigUnit_ModifyTransforms(
    ItemToModify=((Item=(Type=Bone,Name="dyn_braid_MSTR"),Transform=(Rotation=(X=0,Y=0,Z=0,W=1),Translation=(X=0,Y=0,Z=0),Scale3D=(X=1,Y=1,Z=1))),(Item=(Type=Bone,Name="dyn_braid_01"),Transform=(Rotation=(X=0,Y=0,Z=0,W=1),Translation=(X=0,Y=0,Z=0),Scale3D=(X=1,Y=1,Z=1))))
    WeightMinimum=0
    ItemToModify.0.Transform.Rotation=&RigVMFunction_MathQuaternionFromEuler(Euler=(X=0,Y=0,Z=0)).Result
    Weight=@n4.OutputParam
   )
  ]}
]
hierarchy [
 root {loc=<0 0 0>, rot=<0 0 0 1>}
 pelvis {parent=root, loc=<0 -0.7321 92.4218>, rot=<0.0127 0.707 -0.0127 -0.707>}
 spine_01 {parent=pelvis, loc=<5.6002 -0.5128 0>, rot=<0 0 0.034 -0.9994>}
 spine_02 {parent=spine_01, loc=<8.0756 0.0139 0>, rot=<0 0 0.0238 -0.9997>}
 spine_03 {parent=spine_02, loc=<7.0395 0 0>, rot=<0 0 0.0149 -0.9999>}
 spine_04 {parent=spine_03, loc=<8.1023 0 0>, rot=<0 0 -0.0132 -0.9999>}
 spine_05 {parent=spine_04, loc=<10.0141 0 0>, rot=<0 0 -0.0015 -1>}
 neck_01 {parent=spine_05, loc=<10.815 1.9196 0>, rot=<0 0 -0.1864 -0.9825>}
 neck_02 {parent=neck_01, loc=<4.7347 0.3045 0>, rot=<0 0 0.0363 -0.9993>}
 head {parent=neck_02, loc=<5.7632 0.5552 0>, rot=<0 0 0.115 -0.9934>}
 dyn_braid_MSTR {parent=head, loc=<7.9372 -12.4746 -0.1395>, rot=<0.1078 -0.6988 0.6988 -0.1078>}
 dyn_braid_01 {parent=dyn_braid_MSTR, loc=<0 0 0>, rot=<0 0 0 1>}
 dyn_braid_02 {parent=dyn_braid_01, loc=<4.105 0 0.964>, rot=<0 0.0708 0 -0.9975>}
 dyn_braid_03 {parent=dyn_braid_02, loc=<6.5655 0 2.2935>, rot=<0 0.0258 0 -0.9997>}
 dyn_braid_04 {parent=dyn_braid_03, loc=<6.2476 -0.0283 0.0602>, rot=<0 -0.0469 0 -0.9989>}
 dyn_braid_05 {parent=dyn_braid_04, loc=<7.3697 0 -0.201>, rot=<0 0.0127 0 -0.9999>}
 dyn_strand_01 {parent=dyn_braid_02, loc=<0.9516 -1.1034 2.4675>, rot=<-0.1576 0.3127 0.1515 -0.9244>}
 dyn_strand_02 {parent=dyn_braid_02, loc=<5.8445 1.5415 3.342>, rot=<0.443 0.1432 -0.0517 -0.8835>}
 dyn_mohawk_09 {parent=head, loc=<13.2453 -9.3584 0>, rot=<0.647 -0.2853 0.2853 -0.647>}
 dyn_mohawk_08 {parent=head, loc=<14.0489 -5.4428 -2.4774>, rot=<0.617 -0.1977 0.1629 -0.7441>}
 dyn_mohawk_07 {parent=head, loc=<10.8072 -10.2601 2.069>, rot=<0.525 -0.4074 0.3571 -0.6564>}
 dyn_mohawk_06 {parent=head, loc=<15.6233 -2.0928 -0.9082>, rot=<0.702 -0.0848 0.0848 -0.702>}
 dyn_mohawk_05 {parent=head, loc=<14.8544 0.3411 1.8067>, rot=<0.7272 -0.0411 0.0241 -0.6848>}
 dyn_mohawk_04 {parent=head, loc=<14.1112 2.381 -3.2619>, rot=<0.705 0.0551 -0.0551 -0.705>}
 dyn_mohawk_03 {parent=head, loc=<13.2051 5.8237 -3.2619>, rot=<0.698 0.1132 -0.1132 -0.698>}
 dyn_mohawk_02 {parent=head, loc=<13.2 6.441 2.1937>, rot=<0.6957 0.1264 -0.1264 -0.6957>}
 dyn_mohawk_01 {parent=head, loc=<14.4987 -4.7845 2.0489>, rot=<0.6864 -0.1697 0.1697 -0.6864>}
 dyn_hair_neck {parent=head, loc=<2.165 -12.9015 0>, rot=<0.0765 -0.703 0.703 -0.0765>}
 clavicle_l {parent=spine_05, loc=<2.5856 -0.2866 -4.3777>, rot=<0.0465 -0.7289 -0.009 -0.6829>}
 clavicle_r {parent=spine_05, loc=<2.5855 -0.2866 4.3777>, rot=<0.7289 0.0465 -0.6829 0.009>}
 dyn_braid_01_ctrl {controlType=EULER_TRANSFORM, type=Control, loc=<0 0 0>, rot=<0 0 0 1>}
 dyn_braid_01_null {parent=dyn_braid_01_ctrl, type=Null, loc=<0 0 0>, rot=<0 0 0 1>}
 dyn_braid_02_ctrl {controlType=EULER_TRANSFORM, type=Control, loc=<0 0 0>, rot=<0 0 0 1>}
 dyn_braid_02_null {parent=dyn_braid_02_ctrl, type=Null, loc=<0 0 0>, rot=<0 0 0 1>}
 dyn_braid_03_ctrl {controlType=EULER_TRANSFORM, type=Control, loc=<0 0 0>, rot=<0 0 0 1>}
 dyn_braid_03_null {parent=dyn_braid_03_ctrl, type=Null, loc=<0 0 0>, rot=<0 0 0 1>}
 dyn_braid_04_ctrl {controlType=EULER_TRANSFORM, type=Control, loc=<0 0 0>, rot=<0 0 0 1>}
 dyn_braid_04_null {parent=dyn_braid_04_ctrl, type=Null, loc=<0 0 0>, rot=<0 0 0 1>}
 dyn_braid_05_ctrl {controlType=EULER_TRANSFORM, type=Control, loc=<0 0 0>, rot=<0 0 0 1>}
 dyn_braid_05_null {parent=dyn_braid_05_ctrl, type=Null, loc=<0 0 0>, rot=<0 0 0 1>}
]
variables [
 CrouchingMoving {type=bool, default=False}
 dyn_Hair_Back_ctrls {
  type="TArray<FRigElementKey>"
  default=((Type=Control,Name="dyn_Hair_BackL_02"),(Type=Control,Name="dyn_Hair_BackL_03"),(Type=Control,Name="dyn_Hair_BackL_04"),(Type=Control,Name="dyn_Hair_BackL_05"),(Type=Control,Name="dyn_Hair_BackL_06"))
 }
 dyn_Hair_Back_pose {type=FRigPose, default=(Elements=,HierarchyTopologyVersion=-1,PoseHash=-1)}
]
