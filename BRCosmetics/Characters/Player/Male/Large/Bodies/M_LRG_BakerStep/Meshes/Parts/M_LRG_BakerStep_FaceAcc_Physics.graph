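# Physics asset for the face accessory of the large male BakerStep player body, rigging a dynamic braid.
# Five capsule bodies chain from a small master body at the neck, with limited swing and tight twist at each link.
# Neck and upper-spine capsules, including three fanned across spine_05, approximate the torso, presumably for braid collision.
# Constraints joining the master, neck and spine bodies leave swing and twist free.
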PhysicsAsset /BRCosmetics/Characters/Player/Male/Large/Bodies/M_LRG_BakerStep/Meshes/Parts/M_LRG_BakerStep_FaceAcc_Physics
bodies {
 dyn_braid_01 [{shapeType=Capsule, length=4.9117, radius=1.4801, shapeName=dyn_braid_01_capsule, center=<1 0 1>, rotation=<rot 90 0 0>}]
 dyn_braid_02 [{shapeType=Capsule, length=5.4999, radius=2.6971, shapeName=dyn_braid_02_capsule, center=<1.3308 0 2>, rotation=<rot 70.0001 -180 -180>}]
 dyn_braid_03 [{shapeType=Capsule, length=4.237, radius=1.9541, shapeName=dyn_braid_03_capsule, center=<2 0 0.5>, rotation=<rot 90 0 0>}]
 dyn_braid_04 [{shapeType=Capsule, length=4.9175, radius=1.275, shapeName=dyn_braid_04_capsule, center=<3.1177 0 0>, rotation=<rot 90 0 0>}]
 dyn_braid_05 [{shapeType=Capsule, length=3.6374, radius=2.4896, shapeName=dyn_braid_05_capsule, center=<3.0818 0 0>, rotation=<rot 90 0 0>}]
 dyn_braid_MSTR [{shapeType=Capsule, length=0.505, radius=0.505, shapeName=dyn_braid_MSTR_capsule, rotation=<rot 0 -90 0>}]
 neck_02 [{shapeType=Capsule, length=15.2571, radius=11.563, shapeName=neck_02_capsule, center=<-1.2876 -3.648 0>, rotation=<rot 0 -90 0>}]
 spine_04 [{shapeType=Capsule, length=27.841, radius=15.5893, shapeName=spine_04_capsule, center=<3.9704 -0.6423 0>, rotation=<rot 0 -90 0>}]
 spine_05 [
  {shapeType=Capsule, length=14.5745, radius=16.5557, shapeName=spine_05_capsule, center=<3.7415 -0.5735 0>, rotation=<rot 0 -90 0>}
  {shapeType=Capsule, length=14.5745, radius=16.5557, shapeName=spine_05_capsule, center=<3.7415 -0.5735 -5>, rotation=<rot 15 -90.0002 0>}
  {shapeType=Capsule, length=14.5745, radius=16.5557, shapeName=spine_05_capsule, center=<3.7415 -0.5735 5>, rotation=<rot -15 -90.0009 0>}
 ]
}
constraints [
 {
  bone1Name=dyn_braid_MSTR
  bone2Name=neck_02
  swing1LimitDegrees=45
  swing1Motion=Free
  swing2LimitDegrees=45
  swing2Motion=Free
  twistLimitDegrees=45
  twistMotion=Free
 }
 {bone1Name=dyn_braid_01, bone2Name=dyn_braid_MSTR, swing1LimitDegrees=30, swing2LimitDegrees=15, twistLimitDegrees=5}
 {bone1Name=dyn_braid_02, bone2Name=dyn_braid_01, swing1LimitDegrees=20, swing2LimitDegrees=22, twistLimitDegrees=5}
 {bone1Name=dyn_braid_03, bone2Name=dyn_braid_02, swing1LimitDegrees=22, swing2LimitDegrees=40, twistLimitDegrees=5}
 {bone1Name=dyn_braid_04, bone2Name=dyn_braid_03, swing1LimitDegrees=20, swing2LimitDegrees=45, twistLimitDegrees=5}
 {bone1Name=dyn_braid_05, bone2Name=dyn_braid_04, swing1LimitDegrees=20, swing2LimitDegrees=45, twistLimitDegrees=5}
 {
  bone1Name=neck_02
  bone2Name=spine_05
  swing1LimitDegrees=45
  swing1Motion=Free
  swing2LimitDegrees=45
  swing2Motion=Free
  twistLimitDegrees=45
  twistMotion=Free
 }
 {
  bone1Name=spine_05
  bone2Name=spine_04
  swing1LimitDegrees=45
  swing1Motion=Free
  swing2LimitDegrees=45
  swing2Motion=Free
  twistLimitDegrees=45
  twistMotion=Free
 }
]
