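# Post-copy AnimBP for the BakerStep face accessory, adding secondary motion to braid, mohawk and nose ring.
# The copied parent pose passes through a crouch-aware Control Rig, then braid RigidBody branches chosen by back bling and front-end idle, with additive bone fix-ups.
# AnimDynamics springs then drive mohawk and nose ring.
# The EventGraph tunes RigidBody inertia scales for crouch, skydive and back bling; its deprecated gravity-override outputs are unused here.
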AnimBlueprint /BRCosmetics/Characters/Player/Male/Large/Bodies/M_LRG_BakerStep/Meshes/Parts/M_LRG_BakerStep_FaceAcc_AnimBP
target_skeleton=&/M_LRG_BakerStep_FaceAcc_Skeleton
graphs [
 AnimGraph {nodes [
   Output Pose(Result=Use cached pose 'RBAN Cache'()) {layerGroup=None}
   n1 = Control Rig!(Source=Copy Pose From Mesh() {bUseAttachedParent=true}, Alpha, CrouchingMoving:bool=Get CrouchMoving() -> (CrouchMoving:bool)) -> (Pose) {
    alphaBoolBlend {blendOption=Linear}
    controlRigAssetReference {blueprintRigClass=&/M_LRG_BakerStep_FaceAcc_CtrlRig.M_LRG_BakerStep_FaceAcc_CtrlRig_C}
   }
   Save cached pose 'Main Cache'(Pose=@n1) {cacheName=Main Cache}
   n2 = Get AccScaleOverride() -> (AccScaleOverride:Vector)
   n3 = Get VelScaleOverride() -> (VelScaleOverride:Vector)
   n4 = RigidBody!(
    ComponentPose=Local To Component(LocalPose=Use cached pose 'Main Cache'())
    Alpha
    OverrideWorldGravity:Vector="0,0,0"
    ExternalForce
    ComponentLinearAccScale:Vector=@n2
    ComponentLinearVelScale:Vector=@n3
    SimSpaceSettings:Sim Space Settings Structure=(WorldAlpha=0.7,VelocityScaleZ=1,DampingAlpha=1,MaxLinearVelocity=750,MaxAngularVelocity=5,MaxLinearAcceleration=500,MaxAngularAcceleration=3,ExternalLinearDragV=(X=0,Y=0,Z=0),ExternalLinearVelocity=(X=0,Y=0,Z=0),ExternalAngularVelocity=(X=0,Y=0,Z=0))
   ) -> (Pose) {
    bClampLinearTranslationLimitToRefPose=true
    bOverrideWorldGravity=true
    overridePhysicsAsset=&/M_LRG_BakerStep_FaceAcc_Physics
    simulationSpace=BaseBoneSpace
    alphaBoolBlend {blendOption=Linear}
    baseBoneRef {boneName=spine_05}
    componentAppliedLinearAccClamp=<2750 2750 2750>
    componentLinearAccScale=<0.4 0.4 0.4>
    componentLinearVelScale=<0.4 0.4 0.4>
    simSpaceSettings {maxAngularAcceleration=3, maxAngularVelocity=5, maxLinearAcceleration=500, maxLinearVelocity=750, worldAlpha=0.7}
   }
   n5 = RigidBody(
    ComponentPose=Local To Component(LocalPose=Use cached pose 'Main Cache'())
    OverrideWorldGravity="0,0,0"
    ComponentLinearAccScale=@n2
    ComponentLinearVelScale=@n3
    SimSpaceSettings=(WorldAlpha=0.7,VelocityScaleZ=1,DampingAlpha=1,MaxLinearVelocity=750,MaxAngularVelocity=5,MaxLinearAcceleration=500,MaxAngularAcceleration=3,ExternalLinearDragV=(X=0,Y=0,Z=0),ExternalLinearVelocity=(X=0,Y=0,Z=0),ExternalAngularVelocity=(X=0,Y=0,Z=0))
   ) {
    bClampLinearTranslationLimitToRefPose=true
    bOverrideWorldGravity=true
    overridePhysicsAsset=&/M_LRG_BakerStep_FaceAcc_FrontEnd_Physics
    simulationSpace=BaseBoneSpace
    alphaBoolBlend {blendOption=Linear}
    baseBoneRef {boneName=spine_05}
    componentAppliedLinearAccClamp=<2750 2750 2750>
    componentLinearAccScale=<0.4 0.4 0.4>
    componentLinearVelScale=<0.4 0.4 0.4>
    simSpaceSettings {maxAngularAcceleration=3, maxAngularVelocity=5, maxLinearAcceleration=500, maxLinearVelocity=750, worldAlpha=0.7}
   }
   n6 = "Constraint - dyn_braid_02 (PP)"!(ComponentPose=@n5, Alpha=0.2, ConstraintWeights_0=0.5, ConstraintWeights_1:float=0.5) -> (Pose) {
    alphaBoolBlend {blendOption=Linear}
    boneToModify {boneName=dyn_braid_02}
    constraintSetup [{targetBone {boneName=neck_02}}, {targetBone {boneName=head}}]
    constraintWeights [1,1]
   }
   n7 = "Transform (Modify) Bone - Bone: dyn_braid_MSTR"(ComponentPose=@n6, Alpha=1, Translation="-1,0,0.5", Rotation="-5,0,0", Scale="1,1,1") {
    rotationMode=BMM_Additive
    rotationSpace=BCS_BoneSpace
    translationMode=BMM_Additive
    translationSpace=BCS_BoneSpace
    alphaBoolBlend {blendOption=Linear}
    boneToModify {boneName=dyn_braid_MSTR}
    rotation {pitch=-5}
    translation {x=-1, z=0.5}
   }
   n8 = "Transform (Modify) Bone - Bone: dyn_braid_03"(ComponentPose=@n7, Alpha=1, Translation="0,0,-1", Rotation="12.5,0,10", Scale="1,1,1") {
    rotationMode=BMM_Additive
    rotationSpace=BCS_BoneSpace
    translationMode=BMM_Additive
    translationSpace=BCS_BoneSpace
    alphaBoolBlend {blendOption=Linear}
    boneToModify {boneName=dyn_braid_03}
    rotation {pitch=12.5, roll=10}
    translation {z=-1}
   }
   n9 = Blend Poses by bool(
    BlendPose_0=Component To Local(ComponentPose=@n8)
    BlendPose_1=Component To Local(ComponentPose=@n4)
    bActiveValue=Get FrontEnd_Idle() -> (FrontEnd_Idle:bool)
   ) {blendType=Linear}
   n10 = RigidBody(
    ComponentPose=Local To Component(LocalPose=Use cached pose 'Main Cache'())
    OverrideWorldGravity="0,0,-980"
    ComponentLinearAccScale=@n2
    ComponentLinearVelScale=@n3
    SimSpaceSettings=(WorldAlpha=0.7,VelocityScaleZ=1,DampingAlpha=1,MaxLinearVelocity=750,MaxAngularVelocity=5,MaxLinearAcceleration=500,MaxAngularAcceleration=3,ExternalLinearDragV=(X=0,Y=0,Z=0),ExternalLinearVelocity=(X=0,Y=0,Z=0),ExternalAngularVelocity=(X=0,Y=0,Z=0))
   ) {
    bClampLinearTranslationLimitToRefPose=true
    bOverrideWorldGravity=true
    overridePhysicsAsset=&M_LRG_BakerStep_FaceAcc_Physics
    simulationSpace=BaseBoneSpace
    alphaBoolBlend {blendOption=Linear}
    baseBoneRef {boneName=spine_05}
    componentAppliedLinearAccClamp=<2750 2750 2750>
    componentLinearAccScale=<0.4 0.4 0.4>
    componentLinearVelScale=<0.4 0.4 0.4>
    simSpaceSettings {maxAngularAcceleration=3, maxAngularVelocity=5, maxLinearAcceleration=500, maxLinearVelocity=750, worldAlpha=0.7}
   }
   n11 = "Transform (Modify) Bone - Bone: dyn_braid_MSTR"!(
    ComponentPose=@n10
    bAlphaBoolEnabled:bool=Get WearingBackBling() -> (WearingBackBling:bool)
    Translation="0,0,0.5"
    Rotation="-5.000,0.000,0.000"
    Scale="1,1,1"
   ) -> (Pose) {
    alphaInputType=Bool
    rotationMode=BMM_Additive
    rotationSpace=BCS_BoneSpace
    translationMode=BMM_Additive
    translationSpace=BCS_BoneSpace
    alphaBoolBlend {blendOption=Linear}
    boneToModify {boneName=dyn_braid_MSTR}
    rotation=<rot -5 0 0>
    translation {z=0.5}
   }
   n12 = Component To Local(ComponentPose=@n11)
   n13 = Get WearingBackBling()
   n14 = Anim Dynamics - Bone: dyn_mohawk_01!(
    ComponentPose=Local To Component(LocalPose=Blend Poses by bool(BlendPose_0=@n12, BlendPose_1=@n9, bActiveValue=@n13) {blendType=Linear})
    Alpha=1
    GravityOverride:Vector="0,0,0"
    ExternalForce="0,0,0"
   ) -> (Pose) {
    angularDampingOverride=1
    angularSpringConstant=6000
    bAngularSpring=true
    bOverrideAngularDamping=true
    bUseGravityOverride=true
    simulationSpace=RootRelative
    alphaBoolBlend {blendOption=Linear}
    boundBone {boneName=dyn_mohawk_01}
    componentAppliedLinearAccClamp=<1000 1000 1000>
    componentLinearAccScale=<0.3 0.3 0.3>
    componentLinearVelScale=<0.4 0.4 0.4>
    physicsBodyDefinitions [{
      boundBone {boneName=dyn_mohawk_01}
      boxExtents=<20 20 20>
      constraintSetup {angularLimitsMax=<4 20 20>, angularLimitsMin=<-4 -20 0>, angularTarget=<1 0 0>}
      localJointOffset=<0.5 0 0>
     }]
   }
   n15 = Anim Dynamics - Bone: dyn_mohawk_02!(ComponentPose=@n14, Alpha=1, GravityOverride:Vector="0,0,0", ExternalForce="0,0,0") -> (Pose) {
    angularDampingOverride=1
    angularSpringConstant=6000
    bAngularSpring=true
    bOverrideAngularDamping=true
    bUseGravityOverride=true
    simulationSpace=RootRelative
    alphaBoolBlend {blendOption=Linear}
    boundBone {boneName=dyn_mohawk_02}
    componentAppliedLinearAccClamp=<1000 1000 1000>
    componentLinearAccScale=<0.3 0.3 0.3>
    componentLinearVelScale=<0.4 0.4 0.4>
    physicsBodyDefinitions [{
      boundBone {boneName=dyn_mohawk_02}
      boxExtents=<20 20 20>
      constraintSetup {angularLimitsMax=<4 20 0>, angularLimitsMin=<-4 -20 -20>, angularTarget=<1 0 0>}
      localJointOffset=<0.3 0 0>
     }]
   }
   n16 = Anim Dynamics - Bone: dyn_mohawk_03!(ComponentPose=@n15, Alpha=1, GravityOverride:Vector="0,0,0", ExternalForce="0,0,0") -> (Pose) {
    angularDampingOverride=1
    angularSpringConstant=6000
    bAngularSpring=true
    bOverrideAngularDamping=true
    bUseGravityOverride=true
    simulationSpace=RootRelative
    alphaBoolBlend {blendOption=Linear}
    boundBone {boneName=dyn_mohawk_03}
    componentAppliedLinearAccClamp=<1000 1000 1000>
    componentLinearAccScale=<0.3 0.3 0.3>
    componentLinearVelScale=<0.4 0.4 0.4>
    physicsBodyDefinitions [{
      boundBone {boneName=dyn_mohawk_03}
      boxExtents=<20 20 20>
      constraintSetup {angularLimitsMax=<4 20 0>, angularLimitsMin=<-4 -20 -20>, angularTarget=<1 0 0>}
      localJointOffset=<0.4 0 0>
     }]
   }
   n17 = Anim Dynamics - Bone: dyn_mohawk_04!(ComponentPose=@n16, Alpha=1, GravityOverride:Vector="0,0,0", ExternalForce="0,0,0") -> (Pose) {
    angularDampingOverride=1
    angularSpringConstant=6000
    bAngularSpring=true
    bOverrideAngularDamping=true
    bUseGravityOverride=true
    simulationSpace=RootRelative
    alphaBoolBlend {blendOption=Linear}
    boundBone {boneName=dyn_mohawk_04}
    componentAppliedLinearAccClamp=<1000 1000 1000>
    componentLinearAccScale=<0.3 0.3 0.3>
    componentLinearVelScale=<0.4 0.4 0.4>
    physicsBodyDefinitions [{
      boundBone {boneName=dyn_mohawk_04}
      boxExtents=<20 20 20>
      constraintSetup {angularLimitsMax=<4 20 0>, angularLimitsMin=<-4 -20 -20>, angularTarget=<1 0 0>}
      localJointOffset=<0.5 0 0>
     }]
   }
   n18 = Anim Dynamics - Bone: dyn_mohawk_05!(ComponentPose=@n17, Alpha=1, GravityOverride:Vector="0,0,0", ExternalForce="0,0,0") -> (Pose) {
    angularDampingOverride=1
    angularSpringConstant=6000
    bAngularSpring=true
    bOverrideAngularDamping=true
    bUseGravityOverride=true
    simulationSpace=RootRelative
    alphaBoolBlend {blendOption=Linear}
    boundBone {boneName=dyn_mohawk_05}
    componentAppliedLinearAccClamp=<1000 1000 1000>
    componentLinearAccScale=<0.3 0.3 0.3>
    componentLinearVelScale=<0.4 0.4 0.4>
    physicsBodyDefinitions [{
      boundBone {boneName=dyn_mohawk_05}
      boxExtents=<20 20 20>
      constraintSetup {angularLimitsMax=<4 20 20>, angularLimitsMin=<-4 -20 0>, angularTarget=<1 0 0>}
      localJointOffset=<0.4 0 0>
     }]
   }
   n19 = Anim Dynamics - Bone: dyn_mohawk_06!(ComponentPose=@n18, Alpha=1, GravityOverride:Vector="0,0,0", ExternalForce="0,0,0") -> (Pose) {
    angularDampingOverride=1
    angularSpringConstant=6000
    bAngularSpring=true
    bOverrideAngularDamping=true
    bUseGravityOverride=true
    simulationSpace=RootRelative
    alphaBoolBlend {blendOption=Linear}
    boundBone {boneName=dyn_mohawk_06}
    componentAppliedLinearAccClamp=<1000 1000 1000>
    componentLinearAccScale=<0.3 0.3 0.3>
    componentLinearVelScale=<0.4 0.4 0.4>
    physicsBodyDefinitions [{
      boundBone {boneName=dyn_mohawk_06}
      boxExtents=<20 20 20>
      constraintSetup {angularLimitsMax=<4 20 20>, angularLimitsMin=<-4 -20 0>, angularTarget=<1 0 0>}
      localJointOffset=<0.5 0 0>
     }]
   }
   n20 = Anim Dynamics - Bone: dyn_mohawk_07!(ComponentPose=@n19, Alpha=1, GravityOverride:Vector="0,0,0", ExternalForce="0,0,0") -> (Pose) {
    angularDampingOverride=1
    angularSpringConstant=6000
    bAngularSpring=true
    bOverrideAngularDamping=true
    bUseGravityOverride=true
    simulationSpace=RootRelative
    alphaBoolBlend {blendOption=Linear}
    boundBone {boneName=dyn_mohawk_07}
    componentAppliedLinearAccClamp=<1000 1000 1000>
    componentLinearAccScale=<0.3 0.3 0.3>
    componentLinearVelScale=<0.4 0.4 0.4>
    physicsBodyDefinitions [{
      boundBone {boneName=dyn_mohawk_07}
      boxExtents=<20 20 20>
      constraintSetup {angularLimitsMax=<4 20 20>, angularLimitsMin=<-4 -20 0>, angularTarget=<1 0 0>}
      localJointOffset=<0.3 0 0>
     }]
   }
   n21 = Anim Dynamics - Bone: dyn_mohawk_08!(ComponentPose=@n20, Alpha=1, GravityOverride:Vector="0,0,0", ExternalForce="0,0,0") -> (Pose) {
    angularDampingOverride=1
    angularSpringConstant=6000
    bAngularSpring=true
    bOverrideAngularDamping=true
    bUseGravityOverride=true
    simulationSpace=RootRelative
    alphaBoolBlend {blendOption=Linear}
    boundBone {boneName=dyn_mohawk_08}
    componentAppliedLinearAccClamp=<1000 1000 1000>
    componentLinearAccScale=<0.3 0.3 0.3>
    componentLinearVelScale=<0.4 0.4 0.4>
    physicsBodyDefinitions [{
      boundBone {boneName=dyn_mohawk_08}
      boxExtents=<20 20 20>
      constraintSetup {angularLimitsMax=<4 20 20>, angularLimitsMin=<-4 -20 0>, angularTarget=<1 0 0>}
      localJointOffset=<0.5 0 0>
     }]
   }
   n22 = Anim Dynamics - Bone: dyn_mohawk_09!(ComponentPose=@n21, Alpha=1, GravityOverride:Vector="0,0,0", ExternalForce="0,0,0") -> (Pose) {
    angularDampingOverride=1
    angularSpringConstant=6000
    bAngularSpring=true
    bOverrideAngularDamping=true
    bUseGravityOverride=true
    simulationSpace=RootRelative
    alphaBoolBlend {blendOption=Linear}
    boundBone {boneName=dyn_mohawk_09}
    componentAppliedLinearAccClamp=<1000 1000 1000>
    componentLinearAccScale=<0.3 0.3 0.3>
    componentLinearVelScale=<0.4 0.4 0.4>
    physicsBodyDefinitions [{
      boundBone {boneName=dyn_mohawk_09}
      boxExtents=<20 20 20>
      constraintSetup {angularLimitsMax=<4 20 20>, angularLimitsMin=<-4 -20 0>, angularTarget=<1 0 0>}
      localJointOffset=<0.5 0 0>
     }]
   }
   n23 = Anim Dynamics - Bone: dyn_snout_nosering(ComponentPose=@n22, Alpha=1, ExternalForce="0,0,0") {
    angularDampingOverride=1
    angularSpringConstant=20000
    bAngularSpring=true
    bOverrideAngularDamping=true
    bUseGravityOverride=true
    gravityScale=0.35
    simulationSpace=RootRelative
    alphaBoolBlend {blendOption=Linear}
    boundBone {boneName=dyn_snout_nosering}
    componentAppliedLinearAccClamp=<700 700 700>
    componentLinearAccScale=<1 1 1>
    componentLinearVelScale=<1 1 1>
    gravityOverride {z=-1800}
    physicsBodyDefinitions [{
      boundBone {boneName=dyn_snout_nosering}
      boxExtents=<40 40 40>
      constraintSetup {angularLimitsMax=<20 30 30>, angularLimitsMin=<-20 -15 -30>, angularTarget=<1 0 0>}
      localJointOffset=<3 0 0>
     }]
   }
   Save cached pose 'RBAN Cache'(Pose=Component To Local(ComponentPose=@n23)) {cacheName=RBAN Cache}
  ]}
 EventGraph {nodes [
   Event BlueprintUpdateAnimation()
   n1 = CheckForBackbling()
   > Set isWearingPack(isWearingPack:bool=@n1) -> (Output_Get:bool)
   > Set isWearingCape(isWearingCape=@n1.bIsBackblingLong)
   > Set WearingBackBling(WearingBackBling:bool=OR Boolean(A=Get isWearingPack() -> (isWearingPack:bool), B=Get isWearingCape() -> (isWearingCape:bool))) -> (Output_Get:bool)
   > n2 = Branch(Condition=Get bIsCrouching())
   > n3 = Set AccScaleOverride(AccScaleOverride:Vector="0.1,0.1,0.05") -> (Output_Get:Vector)
   n4 = Branch(execute=[@n7.else, @n8.then, @n3.then], Condition=OR Boolean(A=Get isWearingPack(), B=Get isWearingCape()))
   > Set AccScaleOverride(AccScaleOverride="0, 0, 0")
   > Set VelScaleOverride(VelScaleOverride:Vector="0, 0, 0") -> (Output_Get:Vector)
   Set AccScaleOverride(execute=@n4.else, AccScaleOverride="0.35,0.35,0.35")
   > Set VelScaleOverride(VelScaleOverride="0.45,0.45,0.45")
   > n5 = Gravity Override State Function DEPRECATED(
    CharacterPartAnimInstance=Self-Reference()
    GameTimeInSeconds=GetGameTimeInSeconds()
    GravityOverrideParams=Get Gravity Override Params()
    RBAN Params=Get RBAN Params()
   )
   > Set Gravity Override(Gravity Override=@n5)
   > Set RBAN Settings Out(RBAN Settings Out:RBANSetting Struct Structure=@n5.RBAN SettingsOut) -> (Output_Get:RBANSetting Struct Structure)
   n6 = Get bIsPlayingEmote()
   > Set FrontEnd_Idle(FrontEnd_Idle:bool=AND Boolean(A=Get bIsInFrontEnd(), B=NOT Boolean(A=@n6))) -> (Output_Get:bool)
   n7 = Branch(execute=@n2.else, Condition=Get bIsSkyDiving())
   > Set AccScaleOverride(AccScaleOverride="0.1,0.1,0.1")
   > n8 = Set VelScaleOverride(VelScaleOverride="0.1,0.1,0.1")
   Event OnBeginCrouching()
   > n9 = Branch(Condition=AND Boolean(A=Get bIsCrouching(), B=Get bIsMoving2D()))
   > Set CrouchMoving(CrouchMoving:bool=true) -> (Output_Get:bool)
   Set CrouchMoving(execute=@n9.else, CrouchMoving=false)
   Event BlueprintInitializeAnimation()
   > Set Body_Mesh(Body_Mesh=GetOwnerSkeletalMeshComponent())
   Event OnEndCrouching()
   > Set CrouchMoving(CrouchMoving=false)
   Event OnCharacterCustomizationLoaded()
  ]}
]
variables [
 AccScaleOverride {type=Vector}
 backbling alpha {type=number}
 Body_Mesh {type=SkeletalMeshComponent}
 cape alpha {type=number}
 Character Mesh {type=SkeletalMeshComponent}
 CrouchMoving {type=boolean}
 FrontEnd_Idle {type=boolean}
 Gravity Override {type=Vector}
 Gravity Override Params {type=GravityOverrideParamsStruct}
 isWearingCape {type=boolean}
 isWearingPack {type=boolean}
 RBAN Params {
  type=RBANParamsStruct
  default {
   dBNO {simSpaceSettings {
     maxAngularAcceleration=1000
     maxAngularVelocity=3
     maxLinearAcceleration=300
     maxLinearVelocity=700
     velocityScaleZ=0
     externalLinearDragV=<0 0.5 0>
    }}
   dBNO_Carried {simSpaceSettings {
     maxAngularAcceleration=1000
     maxAngularVelocity=3
     maxLinearAcceleration=300
     maxLinearVelocity=700
     velocityScaleZ=0
     externalLinearDragV=<0 0.5 0>
    }}
   emoteOrMelee {simSpaceSettings {
     maxAngularAcceleration=1000
     maxAngularVelocity=3
     maxLinearAcceleration=300
     maxLinearVelocity=700
     velocityScaleZ=0
     externalLinearDragV=<0 0.5 0>
    }}
   front_End {simSpaceSettings {
     maxAngularAcceleration=1000
     maxAngularVelocity=3
     maxLinearAcceleration=300
     maxLinearVelocity=700
     velocityScaleZ=0
     externalLinearDragV=<0 0.5 0>
    }}
   onGround_Crouch_Moving {simSpaceSettings {
     maxAngularAcceleration=1000
     maxAngularVelocity=3
     maxLinearAcceleration=300
     maxLinearVelocity=700
     velocityScaleZ=0
     externalLinearDragV=<0 0.5 0>
    }}
   onGround_Falling {simSpaceSettings {
     maxAngularAcceleration=1000
     maxAngularVelocity=3
     maxLinearAcceleration=300
     maxLinearVelocity=700
     velocityScaleZ=0
     externalLinearDragV=<0 0.5 0>
    }}
   onGround_Moving {simSpaceSettings {
     maxAngularAcceleration=1000
     maxAngularVelocity=3
     maxLinearAcceleration=300
     maxLinearVelocity=700
     velocityScaleZ=0
     externalLinearDragV=<0 0.5 0>
    }}
   onGround_Standing {simSpaceSettings {
     maxAngularAcceleration=1000
     maxAngularVelocity=3
     maxLinearAcceleration=300
     maxLinearVelocity=700
     velocityScaleZ=0
     externalLinearDragV=<0 0.5 0>
    }}
   rail_Grinding {simSpaceSettings {
     maxAngularAcceleration=1000
     maxAngularVelocity=3
     maxLinearAcceleration=300
     maxLinearVelocity=700
     velocityScaleZ=0
     worldAlpha=0.1
     externalLinearDragV=<0 0.5 0>
    }}
   riding_Moving {simSpaceSettings {
     maxAngularAcceleration=1000
     maxAngularVelocity=3
     maxLinearAcceleration=300
     maxLinearVelocity=700
     velocityScaleZ=0
     externalLinearDragV=<0 0.5 0>
    }}
   riding_Moving_Targeting {simSpaceSettings {
     maxAngularAcceleration=1000
     maxAngularVelocity=3
     maxLinearAcceleration=300
     maxLinearVelocity=700
     velocityScaleZ=0
     externalLinearDragV=<0 0.5 0>
    }}
   riding_Standing {simSpaceSettings {
     maxAngularAcceleration=1000
     maxAngularVelocity=3
     maxLinearAcceleration=300
     maxLinearVelocity=700
     velocityScaleZ=0
     externalLinearDragV=<0 0.5 0>
    }}
   skydive_DivingDown {simSpaceSettings {
     maxAngularAcceleration=1000
     maxAngularVelocity=3
     maxLinearAcceleration=300
     maxLinearVelocity=700
     velocityScaleZ=0
     externalLinearDragV=<0 0.5 0>
    }}
   skydive_FloatingDown {
    componentLinearAccScale=<0 0.5 0>
    componentLinearVelScale=<0 0.5 0>
    simSpaceSettings {
     maxAngularAcceleration=1000
     maxAngularVelocity=3
     maxLinearAcceleration=300
     maxLinearVelocity=700
     velocityScaleZ=0
     externalLinearDragV=<0 0.5 0>
    }
   }
   skydive_GoingUp {simSpaceSettings {
     maxAngularAcceleration=1000
     maxAngularVelocity=3
     maxLinearAcceleration=300
     maxLinearVelocity=700
     velocityScaleZ=0
     externalLinearDragV=<0 0.5 0>
    }}
   skydive_Parachuting {
    componentLinearAccScale=<0 0.1 0.3>
    componentLinearVelScale=<0 0.1 0.3>
    simSpaceSettings {
     maxAngularAcceleration=1000
     maxAngularVelocity=3
     maxLinearAcceleration=300
     maxLinearVelocity=700
     velocityScaleZ=0
     externalLinearDragV=<0 0.5 0>
    }
   }
   skydive_Surfing {simSpaceSettings {
     maxAngularAcceleration=1000
     maxAngularVelocity=3
     maxLinearAcceleration=300
     maxLinearVelocity=700
     velocityScaleZ=0
     externalLinearDragV=<0 0.5 0>
    }}
   swimming {simSpaceSettings {
     maxAngularAcceleration=1000
     maxAngularVelocity=3
     maxLinearAcceleration=300
     maxLinearVelocity=700
     velocityScaleZ=0
     externalLinearDragV=<0 0.5 0>
    }}
   tactical_Sprint {simSpaceSettings {
     maxAngularAcceleration=1000
     maxAngularVelocity=3
     maxLinearAcceleration=300
     maxLinearVelocity=700
     velocityScaleZ=0
     externalLinearDragV=<0 0.5 0>
    }}
   ziplining {simSpaceSettings {
     maxAngularAcceleration=1000
     maxAngularVelocity=3
     maxLinearAcceleration=300
     maxLinearVelocity=700
     velocityScaleZ=0
     externalLinearDragV=<0 0.5 0>
    }}
  }
 }
 RBAN Settings Out {
  type=RBANSettingStruct
  default {
   componentLinearAccScale=<0 0 0>
   componentLinearVelScale=<0 0 0>
   simSpaceSettings {
    dampingAlpha=1
    maxAngularAcceleration=1000
    maxAngularVelocity=3
    maxLinearAcceleration=300
    maxLinearVelocity=700
    velocityScaleZ=0
    worldAlpha=1
    externalAngularVelocity=<0 0 0>
    externalLinearDragV=<0 0.5 0>
    externalLinearVelocity=<0 0 0>
   }
  }
 }
 VelScaleOverride {type=Vector}
 WearingBackBling {type=boolean}
]
